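# Physics asset for the M_MED_QuickBurst_Plains male medium player body, setting up capsule collision and joints for its dynamic skirt bones.
# Six two-bone skirt chains (A, B, C on each side) hang from spine_02 with free angular motion at the root.
# Each tip is angle-limited to its root, with the A pair much tighter than B and C.
# Body capsules cover only the pelvis, three spine bones and the right thigh, which has no constraint.
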PhysicsAsset /BRCosmetics/Characters/Player/Male/Medium/Bodies/M_MED_QuickBurst_Plains/Meshes/M_MED_QuickBurst_Plains_Physics
bodies {
 dyn_SkirtA00_l [{shapeType=Capsule, length=4, radius=6.5, shapeName=dyn_SkirtB00_l_capsule, center=<0 1.446 0>}]
 dyn_SkirtA00_r [{shapeType=Capsule, length=4, radius=6.5, shapeName=dyn_SkirtB00_r_capsule, center=<0 1.446 0>}]
 dyn_SkirtA01_l [{shapeType=Capsule, length=8, radius=5.5, shapeName=dyn_SkirtB01_l_capsule, center=<0 2.6607 -0.3255>, rotation=<rot -75 -90.0007 90.0009>}]
 dyn_SkirtA01_r [{shapeType=Capsule, length=8, radius=5.5, shapeName=dyn_SkirtB01_r_capsule, center=<0 4.047 -0.2722>, rotation=<rot 70 -90.0003 -90.0005>}]
 dyn_SkirtB00_l [{shapeType=Capsule, length=4, radius=6.5, shapeName=dyn_SkirtB00_l_capsule, center=<0 1.446 0>}]
 dyn_SkirtB00_r [{shapeType=Capsule, length=4, radius=6.5, shapeName=dyn_SkirtB00_r_capsule, center=<0 1.446 0>}]
 dyn_SkirtB01_l [{shapeType=Capsule, length=4, radius=5.5, shapeName=dyn_SkirtB01_l_capsule, center=<3.9117 2.9344 0>}]
 dyn_SkirtB01_r [{shapeType=Capsule, length=8.5, radius=5.5, shapeName=dyn_SkirtB01_r_capsule, center=<3.9117 2.515 0.676>}]
 dyn_SkirtC00_l [{shapeType=Capsule, length=4, radius=6.5, shapeName=dyn_SkirtC00_l_capsule, center=<0 1.446 0>}]
 dyn_SkirtC00_r [{shapeType=Capsule, length=4, radius=6.5, shapeName=dyn_SkirtC00_r_capsule, center=<0 1.446 0>}]
 dyn_SkirtC01_l [{shapeType=Capsule, length=9.75, radius=5.5, shapeName=dyn_SkirtC01_l_capsule, center=<3.9117 3.1046 0>, rotation=<rot -90 120.5643 -120.565>}]
 dyn_SkirtC01_r [{
   shapeType=Capsule
   length=9.75
   radius=5.5
   shapeName=dyn_SkirtC01_r_capsule
   center=<3.9117 2.058 0.5006>
   rotation=<rot -90 11.6877 -11.6884>
  }]
 pelvis [
  {shapeType=Capsule, length=11.25, radius=11, shapeName=pelvis_capsule, center=<0.1982 -1.2398 0>}
  {shapeType=Capsule, length=15, radius=9, shapeName=pelvis_capsule, center=<-7.9326 -4.2175 0>}
  {shapeType=Capsule, length=5.25, radius=8.25, shapeName=pelvis_capsule, center=<-6.7589 3.2707 -6.3273>, rotation=<rot 0 -0.0002 35.1804>}
 ]
 spine_01 [{shapeType=Capsule, length=8.75, radius=8.5, shapeName=spine_01_capsule, center=<3.1065 0 0>}]
 spine_02 [{shapeType=Capsule, length=6.5, radius=9.75, shapeName=spine_02_capsule, center=<1.3835 0 0>}]
 spine_03 [{shapeType=Capsule, length=7, radius=9.75, shapeName=spine_03_capsule}]
 thigh_r [{
   shapeType=Capsule
   length=31.2287
   radius=10
   shapeName=thigh_r_capsule
   center=<22.4647 -0.8 -0.6954>
   rotation=<rot -7.5491 91.0332 -92.3847>
  }]
}
constraints [
 {bone1Name=spine_01, bone2Name=pelvis, swing1LimitDegrees=45, swing2LimitDegrees=45, twistLimitDegrees=45}
 {bone1Name=spine_03, bone2Name=spine_02, swing1LimitDegrees=45, swing2LimitDegrees=45, twistLimitDegrees=45}
 {bone1Name=dyn_SkirtA01_l, bone2Name=dyn_SkirtA00_l, swing1LimitDegrees=15, swing2LimitDegrees=10, twistLimitDegrees=1}
 {
  bone1Name=dyn_SkirtA00_l
  bone2Name=spine_02
  swing1LimitDegrees=45
  swing1Motion=Free
  swing2LimitDegrees=45
  swing2Motion=Free
  twistLimitDegrees=45
  twistMotion=Free
 }
 {
  bone1Name=dyn_SkirtA00_r
  bone2Name=spine_02
  swing1LimitDegrees=45
  swing1Motion=Free
  swing2LimitDegrees=45
  swing2Motion=Free
  twistLimitDegrees=45
  twistMotion=Free
 }
 {bone1Name=dyn_SkirtA01_r, bone2Name=dyn_SkirtA00_r, swing1LimitDegrees=15, swing2LimitDegrees=10, twistLimitDegrees=1}
 {
  bone1Name=dyn_SkirtB00_l
  bone2Name=spine_02
  swing1LimitDegrees=45
  swing1Motion=Free
  swing2LimitDegrees=45
  swing2Motion=Free
  twistLimitDegrees=45
  twistMotion=Free
 }
 {bone1Name=dyn_SkirtB01_l, bone2Name=dyn_SkirtB00_l, swing1LimitDegrees=60, swing2LimitDegrees=25, twistLimitDegrees=15}
 {
  bone1Name=dyn_SkirtB00_r
  bone2Name=spine_02
  swing1LimitDegrees=45
  swing1Motion=Free
  swing2LimitDegrees=45
  swing2Motion=Free
  twistLimitDegrees=45
  twistMotion=Free
 }
 {bone1Name=dyn_SkirtB01_r, bone2Name=dyn_SkirtB00_r, swing1LimitDegrees=60, swing2LimitDegrees=25, twistLimitDegrees=15}
 {
  bone1Name=dyn_SkirtC00_l
  bone2Name=spine_02
  swing1LimitDegrees=45
  swing1Motion=Free
  swing2LimitDegrees=45
  swing2Motion=Free
  twistLimitDegrees=45
  twistMotion=Free
 }
 {bone1Name=dyn_SkirtC01_l, bone2Name=dyn_SkirtC00_l, swing1LimitDegrees=60, swing2LimitDegrees=25, twistLimitDegrees=15}
 {
  bone1Name=dyn_SkirtC00_r
  bone2Name=spine_02
  swing1LimitDegrees=45
  swing1Motion=Free
  swing2LimitDegrees=45
  swing2Motion=Free
  twistLimitDegrees=45
  twistMotion=Free
 }
 {bone1Name=dyn_SkirtC01_r, bone2Name=dyn_SkirtC00_r, swing1LimitDegrees=60, swing2LimitDegrees=25, twistLimitDegrees=15}
]
