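# Anim Blueprint for the QuickBurst Plains medium male body cosmetic, adding secondary motion to bandana, necklace, collar, lapels, zippers, belt gear and shoulder pad.
# AnimDynamics and RigidBody nodes simulate dangling parts, some fed from a cached 'Presim' pose.
# Weighted multi-bone constraints make attachments follow spine, clavicles, head and limbs.
# Additive bone transforms correct poses using jump, neck-rotation and hand-raise variables; several alternative chains have no input pose connected.
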AnimBlueprint /BRCosmetics/Characters/Player/Male/Medium/Bodies/M_MED_QuickBurst_Plains/Meshes/M_MED_QuickBurst_Plains_AnimBP
target_skeleton=&/M_MED_QuickBurst_Plains_Skeleton
graphs [
 AnimGraph {nodes [
   n1 = Anim Dynamics - Bone: dyn_bandana_01!(ComponentPose, bAlphaBoolEnabled:bool=True, ExternalForce="0,0,0") -> (Pose) {
    alphaInputType=Bool
    angularDampingOverride=1
    bOverrideAngularDamping=true
    bUseGravityOverride=true
    bUsePlanarLimit=false
    bUseSphericalLimits=true
    simulationSpace=RootRelative
    alphaBoolBlend {blendInTime=0.1, blendOption=Linear, blendOutTime=0.2}
    boundBone {boneName=dyn_bandana_01}
    componentAppliedLinearAccClamp=<500 500 500>
    componentLinearAccScale=<0.2 0.2 0.5>
    componentLinearVelScale=<0.3 0.3 0.8>
    gravityOverride {z=-980}
    physicsBodyDefinitions [{
      boundBone {boneName=dyn_bandana_01}
      boxExtents=<8 8 8>
      constraintSetup {angularLimitsMax=<5 15 10>, angularLimitsMin=<-5 -20 -10>, angularTarget=<0 0 0>}
      localJointOffset=<1 0 0>
     }]
    sphericalLimits [{limitRadius=18, drivingBone {boneName=spine_04}, sphereLocalOffset=<0 0 0>}]
   }
   Anim Dynamics - Bone: dyn_bandana_02!(ComponentPose=@n1, bAlphaBoolEnabled:bool=True, ExternalForce="0,0,0") -> (Pose) {
    alphaInputType=Bool
    angularDampingOverride=1
    bOverrideAngularDamping=true
    bUseGravityOverride=true
    bUsePlanarLimit=false
    bUseSphericalLimits=true
    simulationSpace=RootRelative
    alphaBoolBlend {blendInTime=0.1, blendOption=Linear, blendOutTime=0.2}
    boundBone {boneName=dyn_bandana_02}
    componentAppliedLinearAccClamp=<500 500 500>
    componentLinearAccScale=<0.2 0.2 0.5>
    componentLinearVelScale=<0.3 0.3 0.8>
    gravityOverride {z=-980}
    physicsBodyDefinitions [{
      boundBone {boneName=dyn_bandana_02}
      boxExtents=<8 8 8>
      constraintSetup {angularLimitsMax=<5 8 10>, angularLimitsMin=<-5 -20 -10>, angularTarget=<0 0 0>}
      localJointOffset=<1 0 0>
     }]
    sphericalLimits [{limitRadius=18, drivingBone {boneName=spine_04}, sphereLocalOffset=<0 0 0>}]
   }
   n2 = Anim Dynamics - Bone: dyn_bandana_01(bAlphaBoolEnabled=True, ExternalForce="0,0,0") {
    alphaInputType=Bool
    angularDampingOverride=1
    bOverrideAngularDamping=true
    bUseGravityOverride=true
    bUsePlanarLimit=false
    bUseSphericalLimits=true
    simulationSpace=RootRelative
    alphaBoolBlend {blendInTime=0.1, blendOption=Linear, blendOutTime=0.2}
    boundBone {boneName=dyn_bandana_01}
    componentAppliedLinearAccClamp=<500 500 500>
    componentLinearAccScale=<0.2 0.2 0.5>
    componentLinearVelScale=<0.3 0.3 0.8>
    gravityOverride {z=-980}
    physicsBodyDefinitions [{
      boundBone {boneName=dyn_bandana_01}
      boxExtents=<2 8 4>
      constraintSetup {angularLimitsMax=<5 15 10>, angularLimitsMin=<-5 -20 -10>, angularTarget=<0 0 0>}
      localJointOffset=<1 0 0>
     }]
    sphericalLimits [{limitRadius=18, drivingBone {boneName=spine_04}, sphereLocalOffset=<0 0 0>}]
   }
   Anim Dynamics - Bone: dyn_bandana_02(ComponentPose=@n2, bAlphaBoolEnabled=True, ExternalForce="0,0,0") {
    alphaInputType=Bool
    angularDampingOverride=1
    bOverrideAngularDamping=true
    bUseGravityOverride=true
    bUsePlanarLimit=false
    bUseSphericalLimits=true
    simulationSpace=RootRelative
    alphaBoolBlend {blendInTime=0.1, blendOption=Linear, blendOutTime=0.2}
    boundBone {boneName=dyn_bandana_02}
    componentAppliedLinearAccClamp=<500 500 500>
    componentLinearAccScale=<0.2 0.2 0.5>
    componentLinearVelScale=<0.3 0.3 0.8>
    gravityOverride {z=-980}
    physicsBodyDefinitions [{
      boundBone {boneName=dyn_bandana_02}
      boxExtents=<2 8 4>
      constraintSetup {angularLimitsMax=<5 8 10>, angularLimitsMin=<-5 -20 -10>, angularTarget=<0 0 0>}
      localJointOffset=<1 0 0>
     }]
    sphericalLimits [{limitRadius=18, drivingBone {boneName=spine_04}, sphereLocalOffset=<0 0 0>}]
   }
   n3 = Anim Dynamics - Bone: dyn_bandana_01(bAlphaBoolEnabled=True, ExternalForce="0,0,0") {
    alphaInputType=Bool
    angularDampingOverride=1
    bOverrideAngularDamping=true
    bUseGravityOverride=true
    bUsePlanarLimit=false
    simulationSpace=RootRelative
    alphaBoolBlend {blendInTime=0.1, blendOption=Linear, blendOutTime=0.2}
    boundBone {boneName=dyn_bandana_01}
    componentAppliedLinearAccClamp=<500 500 500>
    componentLinearAccScale=<0.2 0.2 0.5>
    componentLinearVelScale=<0.3 0.3 0.8>
    gravityOverride {z=-700}
    physicsBodyDefinitions [{
      boundBone {boneName=dyn_bandana_01}
      boxExtents=<8 8 8>
      constraintSetup {angularLimitsMax=<30 5 60>, angularLimitsMin=<-30 -50 -20>, angularTarget=<0 0 0>}
      localJointOffset=<1 0 0>
     }]
    sphericalLimits [{limitRadius=18, drivingBone {boneName=spine_03}, sphereLocalOffset=<0 0 0>}]
   }
   Anim Dynamics - Bone: dyn_bandana_02(ComponentPose=@n3, bAlphaBoolEnabled=True, ExternalForce="0,0,0") {
    alphaInputType=Bool
    angularDampingOverride=1
    bOverrideAngularDamping=true
    bUseGravityOverride=true
    bUsePlanarLimit=false
    simulationSpace=RootRelative
    alphaBoolBlend {blendInTime=0.1, blendOption=Linear, blendOutTime=0.2}
    boundBone {boneName=dyn_bandana_02}
    componentAppliedLinearAccClamp=<500 500 500>
    componentLinearAccScale=<0.2 0.2 0.5>
    componentLinearVelScale=<0.3 0.3 0.8>
    gravityOverride {z=-700}
    physicsBodyDefinitions [{
      boundBone {boneName=dyn_bandana_02}
      boxExtents=<7 7 7>
      constraintSetup {angularLimitsMax=<30 0 60>, angularLimitsMin=<-30 -50 -20>, angularTarget=<0 0 0>}
      localJointOffset=<1 0 0>
     }]
    sphericalLimits [{limitRadius=18, drivingBone {boneName=spine_03}, sphereLocalOffset=<0 0 0>}]
   }
   Anim Dynamics - Bone: dyn_necklace_01!(ComponentPose, bAlphaBoolEnabled:bool=True, ExternalForce="0,0,0") -> (Pose) {
    alphaInputType=Bool
    angularDampingOverride=0.999
    angularSpringConstant=50
    bAngularSpring=true
    bOverrideAngularDamping=true
    bUseGravityOverride=true
    bUsePlanarLimit=false
    linearDampingOverride=0.9999
    simulationSpace=RootRelative
    alphaBoolBlend {blendInTime=0.3, blendOption=Linear, blendOutTime=0.4}
    boundBone {boneName=dyn_necklace_01}
    componentAppliedLinearAccClamp=<3000 3000 750>
    componentLinearAccScale=<0.4 0.4 0.3>
    componentLinearVelScale=<0.5 0.5 0.4>
    gravityOverride {z=-400}
    physicsBodyDefinitions [{
      boundBone {boneName=dyn_necklace_01}
      boxExtents=<13 13 13>
      constraintSetup {angularLimitsMax=<0 5 25>, angularLimitsMin=<0 -5 0>, angularTarget=<0 0 0>}
      localJointOffset=<1 0 0>
     }]
   }
   n4 = Anim Dynamics - Bone: dyn_collarB_01_r(Alpha=0.5, ExternalForce="0,0,0") {
    angularDampingOverride=0.999
    angularSpringConstant=800
    bAngularSpring=true
    bGravityOverrideInSimSpace=true
    bOverrideAngularDamping=true
    bUseGravityOverride=true
    gravityScale=0.2
    simulationSpace=RootRelative
    alphaBoolBlend {blendOption=Linear}
    boundBone {boneName=dyn_collarB_01_r}
    componentAppliedLinearAccClamp=<500 500 500>
    componentLinearAccScale=<0.5 0.5 0.1>
    componentLinearVelScale=<0.5 0.5 0.1>
    gravityOverride {z=-300}
    physicsBodyDefinitions [{
      sphereCollisionRadius=0
      boundBone {boneName=dyn_collarB_01_r}
      boxExtents=<15 15 15>
      constraintSetup {angularTargetAxis=AxisY, angularLimitsMax=<1 1 15>, angularLimitsMin=<-1 -1 -15>, angularTarget=<0 0 0>}
      localJointOffset=<-5 0.1 1>
     }]
    relativeSpaceBone {boneName=spine_05}
   }
   Anim Dynamics - Bone: dyn_collarA_01_r(ComponentPose=@n4, Alpha=0.5, ExternalForce="0,0,0") {
    angularDampingOverride=0.999
    angularSpringConstant=800
    bAngularSpring=true
    bGravityOverrideInSimSpace=true
    bOverrideAngularDamping=true
    bUseGravityOverride=true
    gravityScale=0.2
    simulationSpace=RootRelative
    alphaBoolBlend {blendOption=Linear}
    boundBone {boneName=dyn_collarA_01_r}
    componentAppliedLinearAccClamp=<500 500 500>
    componentLinearAccScale=<0.5 0.5 0.1>
    componentLinearVelScale=<0.5 0.5 0.1>
    gravityOverride {z=-300}
    physicsBodyDefinitions [{
      sphereCollisionRadius=0
      boundBone {boneName=dyn_collarA_01_r}
      boxExtents=<15 15 15>
      constraintSetup {angularTargetAxis=AxisY, angularLimitsMax=<1 5 25>, angularLimitsMin=<-1 -5 -25>, angularTarget=<0 0 0>}
      localJointOffset=<-5 0.01 1>
     }]
    relativeSpaceBone {boneName=spine_05}
   }
   n5 = Anim Dynamics - Bone: dyn_collarA_01_l(Alpha=0.5, ExternalForce="0,0,0") {
    angularDampingOverride=0.999
    angularSpringConstant=800
    bAngularSpring=true
    bGravityOverrideInSimSpace=true
    bOverrideAngularDamping=true
    bUseGravityOverride=true
    gravityScale=0.2
    simulationSpace=RootRelative
    alphaBoolBlend {blendOption=Linear}
    boundBone {boneName=dyn_collarA_01_l}
    componentAppliedLinearAccClamp=<500 500 500>
    componentLinearAccScale=<0.5 0.5 0.1>
    componentLinearVelScale=<0.5 0.5 0.1>
    gravityOverride {z=-300}
    physicsBodyDefinitions [{
      sphereCollisionRadius=0
      boundBone {boneName=dyn_collarA_01_l}
      boxExtents=<15 15 15>
      constraintSetup {angularTargetAxis=AxisY, angularLimitsMax=<1 5 25>, angularLimitsMin=<-1 -5 -25>, angularTarget=<0 0 0>}
      localJointOffset=<-5 0.01 -1>
     }]
    relativeSpaceBone {boneName=spine_05}
   }
   n6 = Anim Dynamics - Bone: dyn_collarB_01_l(ComponentPose=@n5, Alpha=0.5, ExternalForce="0,0,0") {
    angularDampingOverride=0.999
    angularSpringConstant=800
    bAngularSpring=true
    bGravityOverrideInSimSpace=true
    bOverrideAngularDamping=true
    bUseGravityOverride=true
    gravityScale=0.2
    simulationSpace=RootRelative
    alphaBoolBlend {blendOption=Linear}
    boundBone {boneName=dyn_collarB_01_l}
    componentAppliedLinearAccClamp=<500 500 500>
    componentLinearAccScale=<0.5 0.5 0.1>
    componentLinearVelScale=<0.5 0.5 0.1>
    gravityOverride {z=-300}
    physicsBodyDefinitions [{
      sphereCollisionRadius=0
      boundBone {boneName=dyn_collarB_01_l}
      boxExtents=<15 15 15>
      constraintSetup {angularTargetAxis=AxisY, angularLimitsMax=<1 1 15>, angularLimitsMin=<-1 -1 -15>, angularTarget=<0 0 0>}
      localJointOffset=<-5 0.1 -1>
     }]
    relativeSpaceBone {boneName=spine_05}
   }
   n7 = Anim Dynamics - Bone: dyn_collar_01(ComponentPose=@n6, Alpha=0.5, ExternalForce="0,0,0") {
    angularDampingOverride=0.999
    angularSpringConstant=800
    bAngularSpring=true
    bGravityOverrideInSimSpace=true
    bOverrideAngularDamping=true
    bUseGravityOverride=true
    gravityScale=0.2
    simulationSpace=RootRelative
    alphaBoolBlend {blendOption=Linear}
    boundBone {boneName=dyn_collar_01}
    componentAppliedLinearAccClamp=<500 500 500>
    componentLinearAccScale=<0.5 0.5 0.1>
    componentLinearVelScale=<0.5 0.5 0.1>
    gravityOverride {z=-300}
    physicsBodyDefinitions [{
      sphereCollisionRadius=0
      boundBone {boneName=dyn_collar_01}
      boxExtents=<25 25 25>
      constraintSetup {angularTargetAxis=AxisY, angularLimitsMax=<1 5 25>, angularLimitsMin=<-1 -5 -25>, angularTarget=<0 0 0>}
      localJointOffset=<-5 -1 -0.1>
     }]
    relativeSpaceBone {boneName=spine_05}
   }
   n8 = Anim Dynamics - Bone: dyn_collarA_01_r(ComponentPose=@n7, Alpha=0.5, ExternalForce="0,0,0") {
    angularDampingOverride=0.999
    angularSpringConstant=800
    bAngularSpring=true
    bGravityOverrideInSimSpace=true
    bOverrideAngularDamping=true
    bUseGravityOverride=true
    gravityScale=0.2
    simulationSpace=RootRelative
    alphaBoolBlend {blendOption=Linear}
    boundBone {boneName=dyn_collarA_01_r}
    componentAppliedLinearAccClamp=<500 500 500>
    componentLinearAccScale=<0.5 0.5 0.1>
    componentLinearVelScale=<0.5 0.5 0.1>
    gravityOverride {z=-300}
    physicsBodyDefinitions [{
      sphereCollisionRadius=0
      boundBone {boneName=dyn_collarA_01_r}
      boxExtents=<15 15 15>
      constraintSetup {angularTargetAxis=AxisY, angularLimitsMax=<1 5 25>, angularLimitsMin=<-1 -5 -25>, angularTarget=<0 0 0>}
      localJointOffset=<-5 0.01 1>
     }]
    relativeSpaceBone {boneName=spine_05}
   }
   Anim Dynamics - Bone: dyn_collarB_01_r(ComponentPose=@n8, Alpha=0.5, ExternalForce="0,0,0") {
    angularDampingOverride=0.999
    angularSpringConstant=800
    bAngularSpring=true
    bGravityOverrideInSimSpace=true
    bOverrideAngularDamping=true
    bUseGravityOverride=true
    gravityScale=0.2
    simulationSpace=RootRelative
    alphaBoolBlend {blendOption=Linear}
    boundBone {boneName=dyn_collarB_01_r}
    componentAppliedLinearAccClamp=<500 500 500>
    componentLinearAccScale=<0.5 0.5 0.1>
    componentLinearVelScale=<0.5 0.5 0.1>
    gravityOverride {z=-300}
    physicsBodyDefinitions [{
      sphereCollisionRadius=0
      boundBone {boneName=dyn_collarB_01_r}
      boxExtents=<15 15 15>
      constraintSetup {angularTargetAxis=AxisY, angularLimitsMax=<1 1 15>, angularLimitsMin=<-1 -1 -15>, angularTarget=<0 0 0>}
      localJointOffset=<-5 0.1 1>
     }]
    relativeSpaceBone {boneName=spine_05}
   }
   "Constraint - dyn_kneepad_r (RR)"!(ComponentPose, Alpha=1, ConstraintWeights_0=0.55, ConstraintWeights_1:float=0.45) -> (Pose) {
    alphaBoolBlend {blendOption=Linear}
    boneToModify {boneName=dyn_kneepad_r}
    constraintSetup [{transformType=Rotation, targetBone {boneName=thigh_r}}, {transformType=Rotation, targetBone {boneName=calf_r}}]
    constraintWeights [1,1]
   }
   n9 = Get neck_rot_up() -> (neck_rot_up:double)
   "Constraint - dyn_collarA_01_r (RR)"!(ComponentPose, Alpha=@n9, ConstraintWeights_0=0.8, ConstraintWeights_1:float=0.2) -> (Pose) {
    alphaBoolBlend {blendOption=Linear}
    boneToModify {boneName=dyn_collarA_01_r}
    constraintSetup [{transformType=Rotation, targetBone {boneName=spine_05}}, {transformType=Rotation, targetBone {boneName=head}}]
    constraintWeights [1,1]
   }
   n10 = "Constraint - dyn_collarA_01_l (P)"(Alpha=2, ConstraintWeights_0=1) {
    alphaBoolBlend {blendOption=Linear}
    boneToModify {boneName=dyn_collarA_01_l}
    constraintSetup [{targetBone {boneName=neck_02}}]
    constraintWeights [1]
   }
   n11 = "Constraint - dyn_collarB_01_l (P)"(ComponentPose=@n10, Alpha=2, ConstraintWeights_0=1) {
    alphaBoolBlend {blendOption=Linear}
    boneToModify {boneName=dyn_collarB_01_l}
    constraintSetup [{targetBone {boneName=neck_02}}]
    constraintWeights [1]
   }
   n12 = "Constraint - dyn_collar_01 (P)"(ComponentPose=@n11, Alpha=2, ConstraintWeights_0=1) {
    alphaBoolBlend {blendOption=Linear}
    boneToModify {boneName=dyn_collar_01}
    constraintSetup [{targetBone {boneName=neck_02}}]
    constraintWeights [1]
   }
   n13 = "Constraint - dyn_collarA_01_r (P)"(ComponentPose=@n12, Alpha=2, ConstraintWeights_0=1) {
    alphaBoolBlend {blendOption=Linear}
    boneToModify {boneName=dyn_collarA_01_r}
    constraintSetup [{targetBone {boneName=neck_02}}]
    constraintWeights [1]
   }
   "Constraint - dyn_collarB_01_r (P)"(ComponentPose=@n13, Alpha=2, ConstraintWeights_0=1) {
    alphaBoolBlend {blendOption=Linear}
    boneToModify {boneName=dyn_collarB_01_r}
    constraintSetup [{targetBone {boneName=neck_02}}]
    constraintWeights [1]
   }
   "Constraint - dyn_collarB_01_l (PP)"!(ComponentPose, Alpha=1, ConstraintWeights_0=0.5, ConstraintWeights_1:float=0.5) -> (Pose) {
    alphaBoolBlend {blendOption=Linear}
    boneToModify {boneName=dyn_collarB_01_l}
    constraintSetup [{targetBone {boneName=dyn_collarA_01_l}}, {targetBone {boneName=dyn_collar_01}}]
    constraintWeights [1,1]
   }
   n14 = Get Neck Rot Up Helmet() -> (Neck Rot Up Helmet:float)
   n15 = "Transform (Modify) Bone - Bone: dyn_collar_01"(Alpha=@n14, Translation="0,0,-5", Rotation="0,0,0", Scale="1,1,1") {
    rotationMode=BMM_Additive
    translationMode=BMM_Additive
    alphaBoolBlend {blendInTime=0.7, blendOption=Linear, blendOutTime=0.8}
    boneToModify {boneName=dyn_collar_01}
    translation {z=-5}
   }
   n16 = "Transform (Modify) Bone - Bone: dyn_collarB_01_r"(ComponentPose=@n15, Alpha=@n14, Translation="-5,0,-1", Rotation="0,0,0", Scale="1,1,1") {
    rotationMode=BMM_Additive
    translationMode=BMM_Additive
    alphaBoolBlend {blendInTime=0.7, blendOption=Linear, blendOutTime=0.8}
    boneToModify {boneName=dyn_collarB_01_r}
    translation {x=-5, z=-1}
   }
   n17 = "Transform (Modify) Bone - Bone: dyn_collarA_01_r"(ComponentPose=@n16, Alpha=@n14, Translation="0,0,-2", Rotation="0,0,0", Scale="1,1,1") {
    rotationMode=BMM_Additive
    translationMode=BMM_Additive
    alphaBoolBlend {blendInTime=0.7, blendOption=Linear, blendOutTime=0.8}
    boneToModify {boneName=dyn_collarA_01_r}
    translation {z=-2}
   }
   n18 = "Constraint - dyn_collarA_01_l (RR)"!(ComponentPose=@n17, Alpha=@n14, ConstraintWeights_0=0.1, ConstraintWeights_1:float=0.9) -> (Pose) {
    alphaBoolBlend {blendOption=Linear}
    boneToModify {boneName=dyn_collarA_01_l}
    constraintSetup [{transformType=Rotation, targetBone {boneName=head}}, {transformType=Rotation, targetBone {boneName=neck_01}}]
    constraintWeights [1,1]
   }
   n19 = "Constraint - dyn_collarB_01_l (RR)"!(ComponentPose=@n18, Alpha=@n14, ConstraintWeights_0=0.7, ConstraintWeights_1:float=0.3) -> (Pose) {
    alphaBoolBlend {blendOption=Linear}
    boneToModify {boneName=dyn_collarB_01_l}
    constraintSetup [{transformType=Rotation, targetBone {boneName=head}}, {transformType=Rotation, targetBone {boneName=neck_01}}]
    constraintWeights [1,1]
   }
   n20 = "Constraint - dyn_collar_01 (RR)"!(ComponentPose=@n19, Alpha=@n14, ConstraintWeights_0=1, ConstraintWeights_1:float=0) -> (Pose) {
    alphaBoolBlend {blendOption=Linear}
    boneToModify {boneName=dyn_collar_01}
    constraintSetup [{transformType=Rotation, targetBone {boneName=head}}, {transformType=Rotation, targetBone {boneName=neck_01}}]
    constraintWeights [1,1]
   }
   n21 = "Constraint - dyn_collarA_01_r (RR)"(ComponentPose=@n20, Alpha=@n14, ConstraintWeights_0=0.8, ConstraintWeights_1=0.2) {
    alphaBoolBlend {blendOption=Linear}
    boneToModify {boneName=dyn_collarA_01_r}
    constraintSetup [{transformType=Rotation, targetBone {boneName=head}}, {transformType=Rotation, targetBone {boneName=neck_01}}]
    constraintWeights [1,1]
   }
   "Constraint - dyn_collarB_01_r (RR)"!(ComponentPose=@n21, Alpha=@n14, ConstraintWeights_0=0.7, ConstraintWeights_1:float=0.3) -> (Pose) {
    alphaBoolBlend {blendOption=Linear}
    boneToModify {boneName=dyn_collarB_01_r}
    constraintSetup [{transformType=Rotation, targetBone {boneName=head}}, {transformType=Rotation, targetBone {boneName=neck_01}}]
    constraintWeights [1,1]
   }
   "Constraint - dyn_collar_MSTR (RR)"!(ComponentPose, bAlphaBoolEnabled:bool=True, ConstraintWeights_0=1, ConstraintWeights_1:float=0) -> (Pose) {
    alphaInputType=Bool
    alphaBoolBlend {blendOption=Linear}
    boneToModify {boneName=dyn_collar_MSTR}
    constraintSetup [{transformType=Rotation, targetBone {boneName=head}}, {transformType=Rotation, targetBone {boneName=spine_05}}]
    constraintWeights [1,1]
   }
   n22 = "Constraint - lapel_r (PPPPP)"!(
    ComponentPose
    Alpha=1
    ConstraintWeights_0=0.0064
    ConstraintWeights_1:float=0.1184
    ConstraintWeights_2:float=0.4731
    ConstraintWeights_3:float=0.3856
    ConstraintWeights_4:float=0.0165
   ) -> (Pose) {
    alphaBoolBlend {blendOption=Linear}
    boneToModify {boneName=lapel_r}
    constraintSetup [
     {targetBone {boneName=spine_02}}
     {targetBone {boneName=spine_03}}
     {targetBone {boneName=spine_04}}
     {targetBone {boneName=spine_05}}
     {targetBone {boneName=clavicle_r}}
    ]
    constraintWeights [1,1,1,1,1]
   }
   "Constraint - lapel_l (PPPPP)"!(
    ComponentPose=@n22
    Alpha=1
    ConstraintWeights_0=0.0073
    ConstraintWeights_1:float=0.119
    ConstraintWeights_2:float=0.474
    ConstraintWeights_3:float=0.3847
    ConstraintWeights_4:float=0.015
   ) -> (Pose) {
    alphaBoolBlend {blendOption=Linear}
    boneToModify {boneName=lapel_l}
    constraintSetup [
     {targetBone {boneName=spine_02}}
     {targetBone {boneName=spine_03}}
     {targetBone {boneName=spine_04}}
     {targetBone {boneName=spine_05}}
     {targetBone {boneName=clavicle_l}}
    ]
    constraintWeights [1,1,1,1,1]
   }
   n23 = "Constraint - dyn_collar_01 (RR)"(Alpha=@n9, ConstraintWeights_0=0.8, ConstraintWeights_1=0.2) {
    alphaBoolBlend {blendOption=Linear}
    boneToModify {boneName=dyn_collar_01}
    constraintSetup [{transformType=Rotation, targetBone {boneName=spine_05}}, {transformType=Rotation, targetBone {boneName=head}}]
    constraintWeights [1,1]
   }
   "Constraint - dyn_collarB_01_r (RR)"(ComponentPose=@n23, Alpha=@n9, ConstraintWeights_0=0.8, ConstraintWeights_1=0.2) {
    alphaBoolBlend {blendOption=Linear}
    boneToModify {boneName=dyn_collarB_01_r}
    constraintSetup [{transformType=Rotation, targetBone {boneName=spine_05}}, {transformType=Rotation, targetBone {boneName=head}}]
    constraintWeights [1,1]
   }
   n24 = "Constraint - dyn_zipper_l_MSTR (PPPPPPPP)"!(
    ComponentPose
    Alpha=1
    ConstraintWeights_0=0.0138
    ConstraintWeights_1:float=0.0334
    ConstraintWeights_2:float=0.1692
    ConstraintWeights_3:float=0.2506
    ConstraintWeights_4:float=0.1065
    ConstraintWeights_5:float=0.1837
    ConstraintWeights_6:float=0.0781
    ConstraintWeights_7:float=0.1647
   ) -> (Pose) {
    alphaBoolBlend {blendOption=Linear}
    boneToModify {boneName=dyn_zipper_l_MSTR}
    constraintSetup [
     {targetBone {boneName=pelvis}}
     {targetBone {boneName=spine_01}}
     {targetBone {boneName=spine_02}}
     {targetBone {boneName=spine_03}}
     {targetBone {boneName=spine_04}}
     {targetBone {boneName=clavicle_l}}
     {targetBone {boneName=upperarm_twist_01_l}}
     {targetBone {boneName=dyn_SkirtA01_l}}
    ]
    constraintWeights [1,1,1,1,1,1,1,1]
   }
   "Constraint - dyn_zipper_r_MSTR (PPPPPPP)"!(
    ComponentPose=@n24
    Alpha=1
    ConstraintWeights_0=0.0392
    ConstraintWeights_1:float=0.1758
    ConstraintWeights_2:float=0.2818
    ConstraintWeights_3:float=0.1208
    ConstraintWeights_4:float=0.2079
    ConstraintWeights_5:float=0.0739
    ConstraintWeights_6:float=0.1005
   ) -> (Pose) {
    alphaBoolBlend {blendOption=Linear}
    boneToModify {boneName=dyn_zipper_r_MSTR}
    constraintSetup [
     {targetBone {boneName=spine_01}}
     {targetBone {boneName=spine_02}}
     {targetBone {boneName=spine_03}}
     {targetBone {boneName=spine_04}}
     {targetBone {boneName=clavicle_r}}
     {targetBone {boneName=upperarm_twist_01_r}}
     {targetBone {boneName=dyn_SkirtA01_r}}
    ]
    constraintWeights [1,1,1,1,1,1,1]
   }
   n25 = "Constraint - dyn_zipper_l_MSTR (PPPPPPP)"!(
    ComponentPose
    Alpha=1
    ConstraintWeights_0=0.0332
    ConstraintWeights_1:float=0.1744
    ConstraintWeights_2:float=0.2542
    ConstraintWeights_3:float=0.1087
    ConstraintWeights_4:float=0.177
    ConstraintWeights_5:float=0.071
    ConstraintWeights_6:float=0.1815
   ) -> (Pose) {
    alphaBoolBlend {blendOption=Linear}
    boneToModify {boneName=dyn_zipper_l_MSTR}
    constraintSetup [
     {targetBone {boneName=spine_01}}
     {targetBone {boneName=spine_02}}
     {targetBone {boneName=spine_03}}
     {targetBone {boneName=spine_04}}
     {targetBone {boneName=clavicle_l}}
     {targetBone {boneName=upperarm_twist_01_l}}
     {targetBone {boneName=dyn_SkirtA01_l}}
    ]
    constraintWeights [1,1,1,1,1,1,1]
   }
   "Constraint - dyn_zipper_r_MSTR (PPPPPPP)"(
    ComponentPose=@n25
    Alpha=1
    ConstraintWeights_0=0.0281
    ConstraintWeights_1=0.1529
    ConstraintWeights_2=0.2701
    ConstraintWeights_3=0.1287
    ConstraintWeights_4=0.2077
    ConstraintWeights_5=0.0637
    ConstraintWeights_6=0.1489
   ) {
    alphaBoolBlend {blendOption=Linear}
    boneToModify {boneName=dyn_zipper_r_MSTR}
    constraintSetup [
     {targetBone {boneName=spine_01}}
     {targetBone {boneName=spine_02}}
     {targetBone {boneName=spine_03}}
     {targetBone {boneName=spine_04}}
     {targetBone {boneName=clavicle_r}}
     {targetBone {boneName=upperarm_twist_01_r}}
     {targetBone {boneName=dyn_SkirtA01_r}}
    ]
    constraintWeights [1,1,1,1,1,1,1]
   }
   "Constraint - dyn_chain_AIM (PPPPPPP)"!(
    ComponentPose
    Alpha=1
    ConstraintWeights_0=0.001
    ConstraintWeights_1:float=0.0086
    ConstraintWeights_2:float=0.0887
    ConstraintWeights_3:float=0.8427
    ConstraintWeights_4:float=0.0011
    ConstraintWeights_5:float=0.0289
    ConstraintWeights_6:float=0.0289
   ) -> (Pose) {
    alphaBoolBlend {blendOption=Linear}
    boneToModify {boneName=dyn_chain_AIM}
    constraintSetup [
     {targetBone {boneName=spine_02}}
     {targetBone {boneName=spine_03}}
     {targetBone {boneName=spine_04}}
     {targetBone {boneName=spine_05}}
     {targetBone {boneName=neck_02}}
     {targetBone {boneName=clavicle_l}}
     {targetBone {boneName=clavicle_r}}
    ]
    constraintWeights [1,1,1,1,1,1,1]
   }
   "Constraint - dyn_belt_bags (PPPPPP)"!(
    ComponentPose
    Alpha=1
    ConstraintWeights_0=0.0005
    ConstraintWeights_1:float=0.1801
    ConstraintWeights_2:float=0.2844
    ConstraintWeights_3:float=0.495
    ConstraintWeights_4:float=0.0398
    ConstraintWeights_5:float=0.0002
   ) -> (Pose) {
    alphaBoolBlend {blendOption=Linear}
    boneToModify {boneName=dyn_belt_bags}
    constraintSetup [
     {offsetOption=None, targetBone {boneName=spine_02}}
     {offsetOption=None, targetBone {boneName=spine_03}}
     {offsetOption=None, targetBone {boneName=spine_04}}
     {offsetOption=None, targetBone {boneName=spine_05}}
     {offsetOption=None, targetBone {boneName=clavicle_l}}
     {offsetOption=None, targetBone {boneName=dyn_SkirtC01_l}}
    ]
    constraintWeights [1,1,1,1,1,1]
   }
   n26 = "Transform (Modify) Bone - Bone: dyn_belt_bags"!(ComponentPose, bAlphaBoolEnabled:bool=True, Translation="0,0,0", Rotation="7,-25,5", Scale="1,1,1") -> (Pose) {
    alphaInputType=Bool
    rotationMode=BMM_Additive
    translationMode=BMM_Additive
    alphaBoolBlend {blendInTime=0.1, blendOption=Linear, blendOutTime=0.2}
    boneToModify {boneName=dyn_belt_bags}
    rotation=<rot 7 -25 5>
   }
   "Transform (Modify) Bone - Bone: dyn_belt_bags_01_MSTR"!(
    ComponentPose=@n26
    bAlphaBoolEnabled:bool=Reroute Node(InputPin:bool) -> (OutputPin:bool)
    Translation="0,0,0"
    Rotation="5,-5,5"
    Scale="1,1,1"
   ) -> (Pose) {
    alphaInputType=Bool
    rotationMode=BMM_Additive
    translationMode=BMM_Additive
    alphaBoolBlend {blendInTime=0.1, blendOption=Linear, blendOutTime=0.2}
    boneToModify {boneName=dyn_belt_bags_01_MSTR}
    rotation=<rot 5 -5 5>
   }
   n27 = Get L_Hand rise_y() -> (L_Hand rise_y:float)
   n28 = "Transform (Modify) Bone - Bone: shoulderpad_MSTR"(Alpha=@n27, Translation="2,0,0", Rotation="0,0,-35", Scale="1,1,1") {
    rotationMode=BMM_Additive
    rotationSpace=BCS_BoneSpace
    translationMode=BMM_Additive
    translationSpace=BCS_BoneSpace
    alphaBoolBlend {blendInTime=0.7, blendOption=Linear, blendOutTime=0.8}
    boneToModify {boneName=shoulderpad_MSTR}
    rotation {roll=-35}
    translation {x=2}
   }
   "Transform (Modify) Bone - Bone: dyn_shoulderpad_01_MSTR"(ComponentPose=@n28, Alpha=@n27, Translation="3,0,3", Rotation="-5.000,0.000,-0.000", Scale="1,1,1") {
    rotationMode=BMM_Additive
    rotationSpace=BCS_BoneSpace
    scaleSpace=BCS_BoneSpace
    translationMode=BMM_Additive
    translationSpace=BCS_BoneSpace
    alphaBoolBlend {blendInTime=0.7, blendOption=Linear, blendOutTime=0.8}
    boneToModify {boneName=dyn_shoulderpad_01_MSTR}
    rotation=<rot -5 0.0001 -0.0001>
    translation {x=3, z=3}
   }
   "Transform (Modify) Bone - Bone: dyn_collar_MSTR"(Alpha=1, Translation="0,0,0", Rotation="0,0,-45", Scale="1,1,1") {
    rotationMode=BMM_Additive
    translationMode=BMM_Additive
    alphaBoolBlend {blendInTime=0.7, blendOption=Linear, blendOutTime=0.8}
    boneToModify {boneName=dyn_collar_MSTR}
    rotation {roll=-45}
   }
   n29 = Get Neck Rot Up Helmet()
   n30 = "Transform (Modify) Bone - Bone: dyn_collar_01"(Alpha=@n29, Translation="0,0,-5", Rotation="0,0,-35", Scale="1,1,1") {
    rotationMode=BMM_Additive
    translationMode=BMM_Additive
    alphaBoolBlend {blendInTime=0.7, blendOption=Linear, blendOutTime=0.8}
    boneToModify {boneName=dyn_collar_01}
    rotation {roll=-35}
    translation {z=-5}
   }
   n31 = "Transform (Modify) Bone - Bone: dyn_collarB_01_r"(ComponentPose=@n30, Alpha=@n29, Translation="-5,0,-1", Rotation="0,0,-30", Scale="1,1,1") {
    rotationMode=BMM_Additive
    translationMode=BMM_Additive
    alphaBoolBlend {blendInTime=0.7, blendOption=Linear, blendOutTime=0.8}
    boneToModify {boneName=dyn_collarB_01_r}
    rotation {roll=-30}
    translation {x=-5, z=-1}
   }
   "Transform (Modify) Bone - Bone: dyn_collarB_01_l"(ComponentPose=@n31, Alpha=@n29, Translation="0,0,0", Rotation="-45,0,0", Scale="1,1,1") {
    rotationMode=BMM_Additive
    translationMode=BMM_Additive
    alphaBoolBlend {blendInTime=0.7, blendOption=Linear, blendOutTime=0.8}
    boneToModify {boneName=dyn_collarB_01_l}
    rotation {pitch=-45}
   }
   "Transform (Modify) Bone - Bone: dyn_collar_MSTR"(Alpha=1, Translation="0,0,0", Rotation="20,0,0", Scale="1,1,1") {
    rotationMode=BMM_Additive
    translationMode=BMM_Additive
    alphaBoolBlend {blendInTime=0.7, blendOption=Linear, blendOutTime=0.8}
    boneToModify {boneName=dyn_collar_MSTR}
    rotation {pitch=20}
   }
   "Transform (Modify) Bone - Bone: dyn_key_ring"(Alpha=float * float(B=1.5) {is=title K2Node_PromotableOperator}, Translation="0,0,0", Rotation="-45,0,0", Scale="1,1,1") {
    rotationMode=BMM_Additive
    translationMode=BMM_Additive
    alphaBoolBlend {blendOption=Linear}
    boneToModify {boneName=dyn_key_ring}
    rotation {pitch=-45}
   }
   "Transform (Modify) Bone - Bone: dyn_belt_center"(Alpha=1, Translation="0,3,-3", Rotation="0,0,0", Scale="1,1,1") {
    rotationMode=BMM_Additive
    rotationSpace=BCS_BoneSpace
    translationMode=BMM_Additive
    translationSpace=BCS_BoneSpace
    alphaBoolBlend {blendInTime=0.7, blendOption=Linear, blendOutTime=0.8}
    boneToModify {boneName=dyn_belt_center}
    translation {y=3, z=-3}
   }
   n32 = Get Rand_TEST() -> (Rand_TEST:double)
   n33 = "Transform (Modify) Bone - Bone: shoulderpad_MSTR"(Alpha=@n32, Translation="1,0,0", Rotation="0,0,-35", Scale="1,1,1") {
    rotationMode=BMM_Additive
    rotationSpace=BCS_BoneSpace
    translationMode=BMM_Additive
    translationSpace=BCS_BoneSpace
    alphaBoolBlend {blendInTime=0.7, blendOption=Linear, blendOutTime=0.8}
    boneToModify {boneName=shoulderpad_MSTR}
    rotation {roll=-35}
    translation {x=1}
   }
   "Transform (Modify) Bone - Bone: dyn_shoulderpad_01_MSTR"(ComponentPose=@n33, Alpha=@n32, Translation="2,0,3", Rotation="-5.000,0.000,-0.000", Scale="1,1,1") {
    rotationMode=BMM_Additive
    rotationSpace=BCS_BoneSpace
    scaleSpace=BCS_BoneSpace
    translationMode=BMM_Additive
    translationSpace=BCS_BoneSpace
    alphaBoolBlend {blendInTime=0.7, blendOption=Linear, blendOutTime=0.8}
    boneToModify {boneName=dyn_shoulderpad_01_MSTR}
    rotation=<rot -5 0.0001 -0.0001>
    translation {x=2, z=3}
   }
   "Transform (Modify) Bone - Bone: shoulderpad_MSTR"!(ComponentPose, bAlphaBoolEnabled:bool=True, Translation="0,0,1", Rotation="0,0,0", Scale="1,1,1") -> (Pose) {
    alphaInputType=Bool
    rotationMode=BMM_Additive
    rotationSpace=BCS_BoneSpace
    translationMode=BMM_Additive
    translationSpace=BCS_BoneSpace
    alphaBoolBlend {blendInTime=0.1, blendOption=Linear, blendOutTime=0.2}
    boneToModify {boneName=shoulderpad_MSTR}
    translation {z=1}
   }
   n34 = Get bIsJumping()
   n35 = "Transform (Modify) Bone - Bone: dyn_belt_bags_01_MSTR"(bAlphaBoolEnabled=@n34, Translation="0,0,3", Rotation="0,0,0", Scale="1,1,1") {
    alphaInputType=Bool
    rotationMode=BMM_Additive
    rotationSpace=BCS_BoneSpace
    translationMode=BMM_Additive
    translationSpace=BCS_BoneSpace
    alphaBoolBlend {blendInTime=0.4, blendOption=Linear, blendOutTime=0.8}
    boneToModify {boneName=dyn_belt_bags_01_MSTR}
    translation {z=3}
   }
   "Transform (Modify) Bone - Bone: dyn_belt_bags_02_MSTR"!(ComponentPose=@n35, bAlphaBoolEnabled:bool=@n34, Translation="0,0,3", Rotation="0,0,0", Scale="1,1,1") -> (Pose) {
    alphaInputType=Bool
    rotationMode=BMM_Additive
    rotationSpace=BCS_BoneSpace
    translationMode=BMM_Additive
    translationSpace=BCS_BoneSpace
    alphaBoolBlend {blendInTime=0.4, blendOption=Linear, blendOutTime=0.8}
    boneToModify {boneName=dyn_belt_bags_02_MSTR}
    translation {z=3}
   }
   n36 = Get bIsJumping()
   n37 = "Transform (Modify) Bone - Bone: dyn_collarA_01_r"!(ComponentPose, bAlphaBoolEnabled:bool=@n36, Translation="0,0,0", Rotation="0,-30,50", Scale="1,1,1") -> (Pose) {
    alphaInputType=Bool
    rotationMode=BMM_Additive
    rotationSpace=BCS_BoneSpace
    translationMode=BMM_Additive
    translationSpace=BCS_BoneSpace
    alphaBoolBlend {blendInTime=0.3, blendOption=Custom, blendOutTime=0.8}
    boneToModify {boneName=dyn_collarA_01_r}
    rotation {roll=50, yaw=-30}
   }
   n38 = "Transform (Modify) Bone - Bone: dyn_collarA_02_r"!(ComponentPose=@n37, bAlphaBoolEnabled:bool=@n36, Translation="0,0,0", Rotation="-15,0,15", Scale="1,1,1") -> (Pose) {
    alphaInputType=Bool
    rotationMode=BMM_Additive
    rotationSpace=BCS_BoneSpace
    translationMode=BMM_Additive
    translationSpace=BCS_BoneSpace
    alphaBoolBlend {blendInTime=0.3, blendOption=Custom, blendOutTime=0.8}
    boneToModify {boneName=dyn_collarA_02_r}
    rotation {pitch=-15, roll=15}
   }
   "Transform (Modify) Bone - Bone: dyn_collarB_01_r"!(ComponentPose=@n38, bAlphaBoolEnabled:bool=@n36, Translation="0,0,0", Rotation="0,0,50", Scale="1,1,1") -> (Pose) {
    alphaInputType=Bool
    rotationMode=BMM_Additive
    rotationSpace=BCS_BoneSpace
    translationMode=BMM_Additive
    translationSpace=BCS_BoneSpace
    alphaBoolBlend {blendInTime=0.3, blendOption=Custom, blendOutTime=0.8}
    boneToModify {boneName=dyn_collarB_01_r}
    rotation {roll=50}
   }
   n39 = "Transform (Modify) Bone - Bone: dyn_belt_bags_01_MSTR"(bAlphaBoolEnabled=@n34, Translation="0,0,0", Rotation="15,-15,15", Scale="1,1,1") {
    alphaInputType=Bool
    rotationMode=BMM_Additive
    translationMode=BMM_Additive
    alphaBoolBlend {blendInTime=0.1, blendOption=Linear, blendOutTime=0.8}
    boneToModify {boneName=dyn_belt_bags_01_MSTR}
    rotation=<rot 15 -15 15>
   }
   "Transform (Modify) Bone - Bone: dyn_belt_bags_02_MSTR"(ComponentPose=@n39, bAlphaBoolEnabled=@n34, Translation="0,0,0", Rotation="15,-15,15", Scale="1,1,1") {
    alphaInputType=Bool
    rotationMode=BMM_Additive
    translationMode=BMM_Additive
    alphaBoolBlend {blendInTime=0.1, blendOption=Linear, blendOutTime=0.8}
    boneToModify {boneName=dyn_belt_bags_02_MSTR}
    rotation=<rot 15 -15 15>
   }
   "Transform (Modify) Bone - Bone: dyn_belt_center"!(ComponentPose, bAlphaBoolEnabled:bool=Get bIsJumping(), Translation="1.5,0,0", Rotation="0,0,0", Scale="1,1,1") -> (Pose) {
    alphaInputType=Bool
    rotationMode=BMM_Additive
    rotationSpace=BCS_BoneSpace
    translationMode=BMM_Additive
    translationSpace=BCS_BoneSpace
    alphaBoolBlend {blendInTime=0.1, blendOption=Linear, blendOutTime=0.2}
    boneToModify {boneName=dyn_belt_center}
    translation {x=1.5}
   }
   n40 = Get bIsJumping()
   n41 = "Transform (Modify) Bone - Bone: dyn_bandana_offset"!(ComponentPose, bAlphaBoolEnabled:bool=@n40, Translation="0,0,1", Rotation="15,0,0", Scale="1,1,1") -> (Pose) {
    alphaInputType=Bool
    rotationMode=BMM_Additive
    rotationSpace=BCS_BoneSpace
    translationMode=BMM_Additive
    translationSpace=BCS_BoneSpace
    alphaBoolBlend {blendInTime=0.1, blendOption=Linear, blendOutTime=0.8}
    boneToModify {boneName=dyn_bandana_offset}
    rotation {pitch=15}
    translation {z=1}
   }
   n42 = "Transform (Modify) Bone - Bone: dyn_bandana_01"!(ComponentPose=@n41, bAlphaBoolEnabled:bool=@n40, Translation="0,0,0", Rotation="20,0,0", Scale="1,1,1") -> (Pose) {
    alphaInputType=Bool
    rotationMode=BMM_Additive
    rotationSpace=BCS_BoneSpace
    translationMode=BMM_Additive
    translationSpace=BCS_BoneSpace
    alphaBoolBlend {blendInTime=0.1, blendOption=Linear, blendOutTime=1.1}
    boneToModify {boneName=dyn_bandana_01}
    rotation {pitch=20}
   }
   "Transform (Modify) Bone - Bone: dyn_bandana_02"!(ComponentPose=@n42, bAlphaBoolEnabled:bool=@n40, Translation="0,0,0", Rotation="25,0,0", Scale="1,1,1") -> (Pose) {
    alphaInputType=Bool
    rotationMode=BMM_Additive
    rotationSpace=BCS_BoneSpace
    translationMode=BMM_Additive
    translationSpace=BCS_BoneSpace
    alphaBoolBlend {blendInTime=0.1, blendOption=Linear, blendOutTime=1.2}
    boneToModify {boneName=dyn_bandana_02}
    rotation {pitch=25}
   }
   "Transform (Modify) Bone - Bone: dyn_belt"!(ComponentPose, bAlphaBoolEnabled:bool=Get bIsJumping(), Translation="0,0,5", Rotation="-30,0,0", Scale="1,1,1") -> (Pose) {
    alphaInputType=Bool
    rotationMode=BMM_Additive
    translationMode=BMM_Additive
    alphaBoolBlend {blendInTime=0.7, blendOption=Linear, blendOutTime=0.8}
    boneToModify {boneName=dyn_belt}
    rotation {pitch=-30}
    translation {z=5}
   }
   RigidBody!(ComponentPose, Alpha, OverrideWorldGravity:Vector="0,0,-980", ExternalForce) -> (Pose) {
    bClampLinearTranslationLimitToRefPose=true
    bOverrideWorldGravity=true
    overridePhysicsAsset=&/M_MED_QuickBurst_Plains_Physics
    simulationSpace=BaseBoneSpace
    alphaBoolBlend {blendOption=Linear}
    alphaScaleBiasClamp {bInterpResult=true}
    baseBoneRef {boneName=pelvis}
    componentAppliedLinearAccClamp=<2000 2000 2000>
    componentLinearAccScale=<0.35 0.3 0.1>
    componentLinearVelScale=<0.35 0.3 0.1>
    overrideWorldGravity {z=-980}
   }
   n43 = Get Bandana idle() -> (Bandana idle:bool)
   n44 = Get Keys crouching() -> (Keys crouching:bool)
   n45 = RigidBody!(ComponentPose=Local To Component(LocalPose=Use cached pose 'Presim'()), bAlphaBoolEnabled:bool, ExternalForce) -> (Pose) {
    alphaInputType=Bool
    bOverrideWorldGravity=true
    overridePhysicsAsset=&M_MED_QuickBurst_Plains_Physics
    simulationSpace=BaseBoneSpace
    alphaBoolBlend {blendInTime=0.1, blendOption=Linear, blendOutTime=0.2}
    baseBoneRef {boneName=pelvis}
    componentAppliedLinearAccClamp=<3000 3000 3000>
    componentLinearAccScale=<1 1 1>
    componentLinearVelScale=<1.1 1.1 1.1>
    overrideWorldGravity {z=-980}
   }
   n46 = "Transform (Modify) Bone - Bone: dyn_key_ring"!(ComponentPose=@n45, bAlphaBoolEnabled:bool=True, Translation="0,0,0", Rotation="10,0,0", Scale="1,1,1") -> (Pose) {
    alphaInputType=Bool
    rotationMode=BMM_Additive
    rotationSpace=BCS_BoneSpace
    translationMode=BMM_Additive
    translationSpace=BCS_BoneSpace
    alphaBoolBlend {blendInTime=0.1, blendOption=Linear, blendOutTime=0.2}
    boneToModify {boneName=dyn_key_ring}
    rotation {pitch=10}
   }
   n47 = Component To Local(ComponentPose=@n46)
   n48 = Anim Dynamics - Bone: dyn_key_ring!(ComponentPose=Local To Component(LocalPose=Use cached pose 'Presim'()), bAlphaBoolEnabled:bool=True, ExternalForce="0,0,0") -> (Pose) {
    alphaInputType=Bool
    angularDampingOverride=0.9
    bOverrideAngularDamping=true
    bUseGravityOverride=true
    bUsePlanarLimit=false
    simulationSpace=RootRelative
    alphaBoolBlend {blendInTime=0.1, blendOption=Linear, blendOutTime=0.2}
    boundBone {boneName=dyn_key_ring}
    componentAppliedLinearAccClamp=<3000 3000 3000>
    componentLinearAccScale=<0.2 0.2 0.2>
    componentLinearVelScale=<0.3 0.3 0.3>
    gravityOverride {z=-980}
    physicsBodyDefinitions [{
      boundBone {boneName=dyn_key_ring}
      boxExtents=<5 5 1>
      constraintSetup {angularLimitsMax=<0 20 5>, angularLimitsMin=<0 0 -5>, angularTarget=<0 0 0>}
      localJointOffset=<1 0 0>
     }]
   }
   n49 = Anim Dynamics - Bone: dyn_key_ring(ComponentPose=@n48, bAlphaBoolEnabled=Get bIsJumping(), ExternalForce="0,0,0") {
    alphaInputType=Bool
    angularDampingOverride=0.9999
    bOverrideAngularDamping=true
    bUseGravityOverride=true
    bUsePlanarLimit=false
    simulationSpace=RootRelative
    alphaBoolBlend {blendInTime=0.1, blendOption=Linear, blendOutTime=0.2}
    boundBone {boneName=dyn_key_ring}
    componentAppliedLinearAccClamp=<3000 3000 3000>
    componentLinearAccScale=<0.2 0.2 0.2>
    componentLinearVelScale=<0.3 0.3 0.3>
    gravityOverride {z=-980}
    physicsBodyDefinitions [{
      boundBone {boneName=dyn_key_ring}
      boxExtents=<5 5 1>
      constraintSetup {angularLimitsMax=<0 40 5>, angularLimitsMin=<0 30 -5>, angularTarget=<0 0 0>}
      localJointOffset=<1 0 0>
     }]
   }
   n50 = Component To Local(ComponentPose=@n49)
   n51 = Get bRigidBodyEnabled()
   n52 = Look At - Bone: dyn_key_ring!(
    ComponentPose=Local To Component(LocalPose=Blend Poses by bool(BlendPose_0=@n47, BlendPose_1=@n50, bActiveValue=@n51) {blendType=Linear})
    bAlphaBoolEnabled:bool=Get bIsJumping()
   ) -> (Pose) {
    alphaInputType=Bool
    alphaBoolBlend {blendInTime=0.1, blendOption=Linear, blendOutTime=0.8}
    boneToModify {boneName=dyn_key_ring}
    lookAtLocation=<0 -5 10>
    lookAtTarget {boneReference {boneName=calf_r}}
    lookAt_Axis {axis=<1 0>}
   }
   n53 = Anim Dynamics - Bone: dyn_key_ring(ComponentPose=@n52, bAlphaBoolEnabled=True, ExternalForce="0,0,0") {
    alphaInputType=Bool
    angularDampingOverride=0.99
    bOverrideAngularDamping=true
    bUseGravityOverride=true
    bUsePlanarLimit=false
    bUseSphericalLimits=true
    linearDampingOverride=0.9999
    simulationSpace=RootRelative
    alphaBoolBlend {blendInTime=0.3, blendOption=Linear, blendOutTime=0.4}
    boundBone {boneName=dyn_key_ring}
    componentAppliedLinearAccClamp=<3000 3000 2000>
    componentLinearAccScale=<0.2 0.2 0.35>
    componentLinearVelScale=<0.3 0.3 0.45>
    gravityOverride {z=-850}
    physicsBodyDefinitions [{
      boundBone {boneName=dyn_key_ring}
      boxExtents=<6 6 7>
      constraintSetup {angularLimitsMax=<5 70 5>, angularLimitsMin=<-5 5 -5>, angularTarget=<0 0 0>}
      localJointOffset=<1 0 0>
     }]
    sphericalLimits [{limitRadius=10, drivingBone {boneName=thigh_r}, sphereLocalOffset=<8 0 0>}]
   }
   n54 = Get bIsCrouching()
   n55 = Anim Dynamics - Bone: dyn_keys_01!(ComponentPose=@n53, bAlphaBoolEnabled:bool=@n54, ExternalForce="0,0,0") -> (Pose) {
    alphaInputType=Bool
    angularDampingOverride=0.99
    bOverrideAngularDamping=true
    bUseGravityOverride=true
    bUsePlanarLimit=false
    bUseSphericalLimits=true
    linearDampingOverride=0.9999
    simulationSpace=RootRelative
    alphaBoolBlend {blendInTime=0.1, blendOption=Linear, blendOutTime=0.2}
    boundBone {boneName=dyn_keys_01}
    componentAppliedLinearAccClamp=<3000 3000 1500>
    componentLinearAccScale=<0.2 0.2 0.2>
    componentLinearVelScale=<0.3 0.3 0.3>
    gravityOverride {z=-400}
    physicsBodyDefinitions [{
      boundBone {boneName=dyn_keys_01}
      boxExtents=<5 5 5>
      constraintSetup {angularLimitsMax=<0 40 5>, angularLimitsMin=<0 0 -5>, angularTarget=<0 0 0>}
      localJointOffset=<1 0 0>
     }]
    sphericalLimits [{limitRadius=10, drivingBone {boneName=thigh_r}, sphereLocalOffset=<9 -2 0>}]
   }
   n56 = Anim Dynamics - Bone: dyn_keys_02!(ComponentPose=@n55, bAlphaBoolEnabled:bool=@n54, ExternalForce="0,0,0") -> (Pose) {
    alphaInputType=Bool
    angularDampingOverride=0.99
    bOverrideAngularDamping=true
    bUseGravityOverride=true
    bUsePlanarLimit=false
    bUseSphericalLimits=true
    linearDampingOverride=0.9999
    simulationSpace=RootRelative
    alphaBoolBlend {blendInTime=0.1, blendOption=Linear, blendOutTime=0.2}
    boundBone {boneName=dyn_keys_02}
    componentAppliedLinearAccClamp=<2000 2000 1500>
    componentLinearAccScale=<0.2 0.2 0.2>
    componentLinearVelScale=<0.3 0.3 0.3>
    gravityOverride {z=-450}
    physicsBodyDefinitions [{
      boundBone {boneName=dyn_keys_02}
      boxExtents=<5 5 6>
      constraintSetup {angularLimitsMax=<0 30 5>, angularLimitsMin=<0 0 -5>, angularTarget=<0 0 0>}
      localJointOffset=<1 0 0>
     }]
    sphericalLimits [{limitRadius=10, drivingBone {boneName=thigh_r}, sphereLocalOffset=<9 -2 0>}]
   }
   n57 = Anim Dynamics - Bone: dyn_keys_03!(ComponentPose=@n56, bAlphaBoolEnabled:bool=@n54, ExternalForce="0,0,0") -> (Pose) {
    alphaInputType=Bool
    angularDampingOverride=0.99
    bOverrideAngularDamping=true
    bUseGravityOverride=true
    bUsePlanarLimit=false
    bUseSphericalLimits=true
    linearDampingOverride=0.9999
    simulationSpace=RootRelative
    alphaBoolBlend {blendInTime=0.1, blendOption=Linear, blendOutTime=0.2}
    boundBone {boneName=dyn_keys_03}
    componentAppliedLinearAccClamp=<1000 1000 1500>
    componentLinearAccScale=<0.2 0.2 0.2>
    componentLinearVelScale=<0.3 0.3 0.3>
    gravityOverride {z=-350}
    physicsBodyDefinitions [{
      boundBone {boneName=dyn_keys_03}
      boxExtents=<5 5 5.5>
      constraintSetup {angularLimitsMax=<0 35 5>, angularLimitsMin=<0 0 -5>, angularTarget=<0 0 0>}
      localJointOffset=<1 0 0>
     }]
    sphericalLimits [{limitRadius=10, drivingBone {boneName=thigh_r}, sphereLocalOffset=<9 -2 0>}]
   }
   n58 = Get Not Crouching() -> (Not Crouching:bool)
   n59 = Anim Dynamics - Bone: dyn_keys_01(ComponentPose=@n57, bAlphaBoolEnabled=@n58, ExternalForce="0,0,0") {
    alphaInputType=Bool
    angularDampingOverride=0.99
    bOverrideAngularDamping=true
    bUseGravityOverride=true
    bUsePlanarLimit=false
    bUseSphericalLimits=true
    linearDampingOverride=0.9999
    simulationSpace=RootRelative
    alphaBoolBlend {blendInTime=0.1, blendOption=Linear, blendOutTime=0.2}
    boundBone {boneName=dyn_keys_01}
    componentAppliedLinearAccClamp=<3000 3000 1500>
    componentLinearAccScale=<0.2 0.2 0.2>
    componentLinearVelScale=<0.3 0.3 0.3>
    gravityOverride {z=-400}
    physicsBodyDefinitions [{
      boundBone {boneName=dyn_keys_01}
      boxExtents=<5 5 5>
      constraintSetup {angularLimitsMax=<0 40 5>, angularLimitsMin=<0 0 -5>, angularTarget=<0 0 0>}
      localJointOffset=<1 0 0>
     }]
    sphericalLimits [{limitRadius=10, drivingBone {boneName=thigh_r}, sphereLocalOffset=<8 0 0>}]
   }
   n60 = Anim Dynamics - Bone: dyn_keys_02(ComponentPose=@n59, bAlphaBoolEnabled=@n58, ExternalForce="0,0,0") {
    alphaInputType=Bool
    angularDampingOverride=0.99
    bOverrideAngularDamping=true
    bUseGravityOverride=true
    bUsePlanarLimit=false
    bUseSphericalLimits=true
    linearDampingOverride=0.9999
    simulationSpace=RootRelative
    alphaBoolBlend {blendInTime=0.1, blendOption=Linear, blendOutTime=0.2}
    boundBone {boneName=dyn_keys_02}
    componentAppliedLinearAccClamp=<2000 2000 1500>
    componentLinearAccScale=<0.2 0.2 0.2>
    componentLinearVelScale=<0.3 0.3 0.3>
    gravityOverride {z=-450}
    physicsBodyDefinitions [{
      boundBone {boneName=dyn_keys_02}
      boxExtents=<5 5 6>
      constraintSetup {angularLimitsMax=<0 30 5>, angularLimitsMin=<0 0 -5>, angularTarget=<0 0 0>}
      localJointOffset=<1 0 0>
     }]
    sphericalLimits [{limitRadius=10, drivingBone {boneName=thigh_r}, sphereLocalOffset=<8 0 0>}]
   }
   n61 = Anim Dynamics - Bone: dyn_keys_03(ComponentPose=@n60, bAlphaBoolEnabled=@n58, ExternalForce="0,0,0") {
    alphaInputType=Bool
    angularDampingOverride=0.99
    bOverrideAngularDamping=true
    bUseGravityOverride=true
    bUsePlanarLimit=false
    bUseSphericalLimits=true
    linearDampingOverride=0.9999
    simulationSpace=RootRelative
    alphaBoolBlend {blendInTime=0.1, blendOption=Linear, blendOutTime=0.2}
    boundBone {boneName=dyn_keys_03}
    componentAppliedLinearAccClamp=<1000 1000 1500>
    componentLinearAccScale=<0.2 0.2 0.2>
    componentLinearVelScale=<0.3 0.3 0.3>
    gravityOverride {z=-350}
    physicsBodyDefinitions [{
      boundBone {boneName=dyn_keys_03}
      boxExtents=<5 5 5.5>
      constraintSetup {angularLimitsMax=<0 35 5>, angularLimitsMin=<0 0 -5>, angularTarget=<0 0 0>}
      localJointOffset=<1 0 0>
     }]
    sphericalLimits [{limitRadius=10, drivingBone {boneName=thigh_r}, sphereLocalOffset=<8 0 0>}]
   }
   n62 = "Transform (Modify) Bone - Bone: dyn_key_ring"(ComponentPose=@n61, bAlphaBoolEnabled=@n44, Translation="0,0,0", Rotation="3,0,0", Scale="1,1,1") {
    alphaInputType=Bool
    rotationMode=BMM_Additive
    rotationSpace=BCS_BoneSpace
    translationMode=BMM_Additive
    translationSpace=BCS_BoneSpace
    alphaBoolBlend {blendInTime=0.1, blendOption=Linear, blendOutTime=0.2}
    boneToModify {boneName=dyn_key_ring}
    rotation {pitch=3}
   }
   n63 = "Transform (Modify) Bone - Bone: dyn_keys_01"!(ComponentPose=@n62, bAlphaBoolEnabled:bool=@n44, Translation="0,0,0", Rotation="10,0,0", Scale="1,1,1") -> (Pose) {
    alphaInputType=Bool
    rotationMode=BMM_Additive
    rotationSpace=BCS_BoneSpace
    translationMode=BMM_Additive
    translationSpace=BCS_BoneSpace
    alphaBoolBlend {blendInTime=0.1, blendOption=Linear, blendOutTime=0.2}
    boneToModify {boneName=dyn_keys_01}
    rotation {pitch=10}
   }
   n64 = "Transform (Modify) Bone - Bone: dyn_keys_02"!(ComponentPose=@n63, bAlphaBoolEnabled:bool=@n44, Translation="0,0,0", Rotation="5,0,0", Scale="1,1,1") -> (Pose) {
    alphaInputType=Bool
    rotationMode=BMM_Additive
    rotationSpace=BCS_BoneSpace
    translationMode=BMM_Additive
    translationSpace=BCS_BoneSpace
    alphaBoolBlend {blendInTime=0.1, blendOption=Linear, blendOutTime=0.2}
    boneToModify {boneName=dyn_keys_02}
    rotation {pitch=5}
   }
   n65 = Anim Dynamics - Bone: dyn_bandana_01(ComponentPose=@n64, bAlphaBoolEnabled=@n43, ExternalForce="0,0,0") {
    alphaInputType=Bool
    angularDampingOverride=1
    bOverrideAngularDamping=true
    bUseGravityOverride=true
    bUsePlanarLimit=false
    simulationSpace=RootRelative
    alphaBoolBlend {blendInTime=0.1, blendOption=Linear, blendOutTime=0.2}
    boundBone {boneName=dyn_bandana_01}
    componentAppliedLinearAccClamp=<3000 3000 3000>
    componentLinearAccScale=<0.8 0.8 0.8>
    componentLinearVelScale=<0.9 0.9 0.9>
    gravityOverride {z=-980}
    physicsBodyDefinitions [{
      boundBone {boneName=dyn_bandana_01}
      boxExtents=<8 8 8>
      constraintSetup {angularLimitsMax=<30 5 40>, angularLimitsMin=<-30 -35 -20>, angularTarget=<0 0 0>}
      localJointOffset=<1 0 0>
     }]
    sphericalLimits [{limitRadius=18, drivingBone {boneName=spine_03}, sphereLocalOffset=<0 0 0>}]
   }
   n66 = Anim Dynamics - Bone: dyn_bandana_02(ComponentPose=@n65, bAlphaBoolEnabled=@n43, ExternalForce="0,0,0") {
    alphaInputType=Bool
    angularDampingOverride=1
    bOverrideAngularDamping=true
    bUseGravityOverride=true
    bUsePlanarLimit=false
    simulationSpace=RootRelative
    alphaBoolBlend {blendInTime=0.1, blendOption=Linear, blendOutTime=0.2}
    boundBone {boneName=dyn_bandana_02}
    componentAppliedLinearAccClamp=<3000 3000 3000>
    componentLinearAccScale=<0.8 0.8 0.8>
    componentLinearVelScale=<0.9 0.9 0.9>
    gravityOverride {z=-980}
    physicsBodyDefinitions [{
      boundBone {boneName=dyn_bandana_02}
      boxExtents=<8 8 8>
      constraintSetup {angularLimitsMax=<30 0 40>, angularLimitsMin=<-30 -35 -20>, angularTarget=<0 0 0>}
      localJointOffset=<1 0 0>
     }]
    sphericalLimits [{limitRadius=18, drivingBone {boneName=spine_03}, sphereLocalOffset=<0 0 0>}]
   }
   n67 = Get bIsTargeting()
   n68 = Anim Dynamics - Bone: dyn_bandana_01(ComponentPose=@n66, bAlphaBoolEnabled=@n67, ExternalForce="0,0,0") {
    alphaInputType=Bool
    angularDampingOverride=0.99
    bOverrideAngularDamping=true
    bUseGravityOverride=true
    bUsePlanarLimit=false
    simulationSpace=BoneRelative
    alphaBoolBlend {blendInTime=0.1, blendOption=Linear, blendOutTime=0.2}
    boundBone {boneName=dyn_bandana_01}
    componentAppliedLinearAccClamp=<3000 3000 3000>
    componentLinearAccScale=<0.5 0.5 0.5>
    componentLinearVelScale=<0.8 0.8 0.8>
    gravityOverride {z=-980}
    physicsBodyDefinitions [{
      boundBone {boneName=dyn_bandana_01}
      boxExtents=<8 8 8>
      constraintSetup {angularLimitsMax=<30 5 40>, angularLimitsMin=<-30 -35 -20>, angularTarget=<0 0 0>}
      localJointOffset=<1 0 0>
     }]
    sphericalLimits [{limitRadius=18, drivingBone {boneName=spine_03}, sphereLocalOffset=<0 0 0>}]
   }
   n69 = Anim Dynamics - Bone: dyn_bandana_02(ComponentPose=@n68, bAlphaBoolEnabled=@n67, ExternalForce="0,0,0") {
    alphaInputType=Bool
    angularDampingOverride=0.99
    bOverrideAngularDamping=true
    bUseGravityOverride=true
    bUsePlanarLimit=false
    simulationSpace=BoneRelative
    alphaBoolBlend {blendInTime=0.1, blendOption=Linear, blendOutTime=0.2}
    boundBone {boneName=dyn_bandana_02}
    componentAppliedLinearAccClamp=<3000 3000 3000>
    componentLinearAccScale=<0.5 0.5 0.5>
    componentLinearVelScale=<0.8 0.8 0.8>
    gravityOverride {z=-980}
    physicsBodyDefinitions [{
      boundBone {boneName=dyn_bandana_02}
      boxExtents=<8 8 8>
      constraintSetup {angularLimitsMax=<30 0 40>, angularLimitsMin=<-30 -35 -20>, angularTarget=<0 0 0>}
      localJointOffset=<1 0 0>
     }]
    sphericalLimits [{limitRadius=18, drivingBone {boneName=spine_03}, sphereLocalOffset=<0 0 0>}]
   }
   n70 = Get bIsJumping()
   n71 = Anim Dynamics - Bone: dyn_bandana_01(ComponentPose=@n69, bAlphaBoolEnabled=@n70, ExternalForce="0,0,0") {
    alphaInputType=Bool
    angularDampingOverride=1
    bOverrideAngularDamping=true
    bUseGravityOverride=true
    bUsePlanarLimit=false
    bUseSphericalLimits=true
    simulationSpace=RootRelative
    alphaBoolBlend {blendInTime=0.1, blendOption=Linear, blendOutTime=0.9}
    boundBone {boneName=dyn_bandana_01}
    componentAppliedLinearAccClamp=<3000 3000 3000>
    componentLinearAccScale=<0.8 0.8 0.5>
    componentLinearVelScale=<0.9 0.9 0.6>
    gravityOverride {z=-980}
    physicsBodyDefinitions [{
      boundBone {boneName=dyn_bandana_01}
      boxExtents=<8 8 8>
      constraintSetup {angularLimitsMax=<30 5 40>, angularLimitsMin=<-30 -35 -20>, angularTarget=<0 0 0>}
      localJointOffset=<1 0 0>
     }]
    sphericalLimits [{limitRadius=21, drivingBone {boneName=spine_02}, sphereLocalOffset=<0 0 0>}]
   }
   n72 = Anim Dynamics - Bone: dyn_bandana_02(ComponentPose=@n71, bAlphaBoolEnabled=@n70, ExternalForce="0,0,0") {
    alphaInputType=Bool
    angularDampingOverride=1
    bOverrideAngularDamping=true
    bUseGravityOverride=true
    bUsePlanarLimit=false
    bUseSphericalLimits=true
    simulationSpace=RootRelative
    alphaBoolBlend {blendInTime=0.1, blendOption=Linear, blendOutTime=0.9}
    boundBone {boneName=dyn_bandana_02}
    componentAppliedLinearAccClamp=<3000 3000 3000>
    componentLinearAccScale=<0.8 0.8 0.5>
    componentLinearVelScale=<0.9 0.9 0.6>
    gravityOverride {z=-980}
    physicsBodyDefinitions [{
      boundBone {boneName=dyn_bandana_02}
      boxExtents=<8 8 8>
      constraintSetup {angularLimitsMax=<30 0 40>, angularLimitsMin=<-30 -35 -20>, angularTarget=<0 0 0>}
      localJointOffset=<1 0 0>
     }]
    sphericalLimits [{limitRadius=20.5, drivingBone {boneName=spine_02}, sphereLocalOffset=<0 0 0>}]
   }
   n73 = "Transform (Modify) Bone - Bone: dyn_bandana_offset"(ComponentPose=@n72, bAlphaBoolEnabled=True, Translation="0,0,0", Rotation="5,0,0", Scale="1,1,1") {
    alphaInputType=Bool
    rotationMode=BMM_Additive
    rotationSpace=BCS_BoneSpace
    translationMode=BMM_Additive
    translationSpace=BCS_BoneSpace
    alphaBoolBlend {blendInTime=0.3, blendOption=Linear, blendOutTime=0.8}
    boneToModify {boneName=dyn_bandana_offset}
    rotation {pitch=5}
   }
   n74 = "Constraint - dyn_bandana_MSTR (PPPPP)"!(
    ComponentPose=@n73
    Alpha=1
    ConstraintWeights_0=0.0166
    ConstraintWeights_1:float=0.061
    ConstraintWeights_2:float=0.2113
    ConstraintWeights_3:float=0.4588
    ConstraintWeights_4:float=0.2523
   ) -> (Pose) {
    alphaBoolBlend {blendOption=Linear}
    boneToModify {boneName=dyn_bandana_MSTR}
    constraintSetup [
     {targetBone {boneName=spine_03}}
     {targetBone {boneName=clavicle_r}}
     {targetBone {boneName=upperarm_r}}
     {targetBone {boneName=upperarm_twist_01_r}}
     {targetBone {boneName=upperarm_twist_02_r}}
    ]
    constraintWeights [1,1,1,1,1]
   }
   n75 = "Constraint - lapel_r (PPPPPPPP)"!(
    ComponentPose=@n74
    Alpha=1
    ConstraintWeights_0=0.0057
    ConstraintWeights_1:float=0.0854
    ConstraintWeights_2:float=0.2495
    ConstraintWeights_3:float=0.239
    ConstraintWeights_4:float=0.1472
    ConstraintWeights_5:float=0.0113
    ConstraintWeights_6:float=0.2495
    ConstraintWeights_7:float=0.0125
   ) -> (Pose) {
    alphaBoolBlend {blendOption=Linear}
    boneToModify {boneName=lapel_r}
    constraintSetup [
     {targetBone {boneName=spine_01}}
     {targetBone {boneName=spine_02}}
     {targetBone {boneName=spine_03}}
     {targetBone {boneName=spine_04}}
     {targetBone {boneName=spine_05}}
     {targetBone {boneName=clavicle_l}}
     {targetBone {boneName=clavicle_r}}
     {targetBone {boneName=upperarm_twist_01_r}}
    ]
    constraintWeights [1,1,1,1,1,1,1,1]
   }
   n76 = "Constraint - lapel_l (PPPPPPPP)"!(
    ComponentPose=@n75
    Alpha=1
    ConstraintWeights_0=0.0056
    ConstraintWeights_1:float=0.0857
    ConstraintWeights_2:float=0.2512
    ConstraintWeights_3:float=0.2402
    ConstraintWeights_4:float=0.1451
    ConstraintWeights_5:float=0.2498
    ConstraintWeights_6:float=0.0109
    ConstraintWeights_7:float=0.0114
   ) -> (Pose) {
    alphaBoolBlend {blendOption=Linear}
    boneToModify {boneName=lapel_l}
    constraintSetup [
     {targetBone {boneName=spine_01}}
     {targetBone {boneName=spine_02}}
     {targetBone {boneName=spine_03}}
     {targetBone {boneName=spine_04}}
     {targetBone {boneName=spine_05}}
     {targetBone {boneName=clavicle_l}}
     {targetBone {boneName=upperarm_twist_01_l}}
     {targetBone {boneName=clavicle_r}}
    ]
    constraintWeights [1,1,1,1,1,1,1,1]
   }
   n77 = Get Hand rise_R() -> (Hand rise_R:float)
   n78 = "Transform (Modify) Bone - Bone: lapel_r"(ComponentPose=@n76, Alpha=@n77, Translation="0,0,0", Rotation="10,30,0", Scale="1,1,1") {
    rotationMode=BMM_Additive
    translationMode=BMM_Additive
    alphaBoolBlend {blendInTime=0.7, blendOption=Linear, blendOutTime=0.8}
    boneToModify {boneName=lapel_r}
    rotation {pitch=10, yaw=30}
   }
   n79 = "Transform (Modify) Bone - Bone: lapel_l"(ComponentPose=@n78, Alpha=@n77, Translation="0,0,0", Rotation="0,-20,15", Scale="1,1,1") {
    rotationMode=BMM_Additive
    translationMode=BMM_Additive
    alphaBoolBlend {blendInTime=0.7, blendOption=Linear, blendOutTime=0.8}
    boneToModify {boneName=lapel_l}
    rotation {roll=15, yaw=-20}
   }
   n80 = Anim Dynamics - Bone: dyn_zipper_r!(ComponentPose=@n79, bAlphaBoolEnabled:bool=True, ExternalForce="0,0,0") -> (Pose) {
    alphaInputType=Bool
    angularDampingOverride=0.99
    bOverrideAngularDamping=true
    bUseGravityOverride=true
    bUsePlanarLimit=false
    linearDampingOverride=0.9999
    simulationSpace=RootRelative
    alphaBoolBlend {blendInTime=0.3, blendOption=Linear, blendOutTime=0.4}
    boundBone {boneName=dyn_zipper_r}
    componentAppliedLinearAccClamp=<3000 3000 1500>
    componentLinearAccScale=<0.2 0.2 0.4>
    componentLinearVelScale=<0.3 0.3 0.5>
    gravityOverride {z=-400}
    physicsBodyDefinitions [{
      boundBone {boneName=dyn_zipper_r}
      boxExtents=<5 5 5>
      constraintSetup {angularLimitsMax=<0 35 2>, angularLimitsMin=<0 0 -2>, angularTarget=<0 0 0>}
      localJointOffset=<1 0 0>
     }]
    sphericalLimits [{limitRadius=10, drivingBone {boneName=thigh_r}, sphereLocalOffset=<8 0 0>}]
   }
   n81 = Anim Dynamics - Bone: dyn_zipper_l!(ComponentPose=@n80, bAlphaBoolEnabled:bool=True, ExternalForce="0,0,0") -> (Pose) {
    alphaInputType=Bool
    angularDampingOverride=0.99
    bOverrideAngularDamping=true
    bUseGravityOverride=true
    bUsePlanarLimit=false
    linearDampingOverride=0.9999
    simulationSpace=RootRelative
    alphaBoolBlend {blendInTime=0.3, blendOption=Linear, blendOutTime=0.4}
    boundBone {boneName=dyn_zipper_l}
    componentAppliedLinearAccClamp=<3000 3000 1500>
    componentLinearAccScale=<0.2 0.2 0.4>
    componentLinearVelScale=<0.3 0.3 0.5>
    gravityOverride {z=-400}
    physicsBodyDefinitions [{
      boundBone {boneName=dyn_zipper_l}
      boxExtents=<8 5 5>
      constraintSetup {angularLimitsMax=<0 0 2>, angularLimitsMin=<0 -35 -2>, angularTarget=<0 0 0>}
      localJointOffset=<1 0 0>
     }]
    sphericalLimits [{limitRadius=10, drivingBone {boneName=thigh_r}, sphereLocalOffset=<8 0 0>}]
   }
   n82 = "Transform (Modify) Bone - Bone: dyn_necklace_02"(
    ComponentPose=@n81
    Alpha="Get Necklace  ScubaBond translate"() -> ("Necklace  ScubaBond translate":float)
    Translation="2,0,0"
    Rotation="0,0,0"
    Scale="1,1,1"
   ) {
    rotationMode=BMM_Additive
    rotationSpace=BCS_BoneSpace
    translationMode=BMM_Additive
    translationSpace=BCS_BoneSpace
    alphaBoolBlend {blendOption=Linear}
    boneToModify {boneName=dyn_necklace_02}
    translation {x=2}
   }
   n83 = Anim Dynamics - Bone: dyn_necklace_01!(ComponentPose=@n82, Alpha:float=Get Necklace Alpha() -> (Necklace Alpha:float), ExternalForce="0,0,0") -> (Pose) {
    angularDampingOverride=0.9999
    angularSpringConstant=50
    bOverrideAngularDamping=true
    bUseGravityOverride=true
    bUsePlanarLimit=false
    linearDampingOverride=0.9999
    simulationSpace=RootRelative
    alphaBoolBlend {blendInTime=0.1, blendOption=Linear, blendOutTime=0.2}
    boundBone {boneName=dyn_necklace_01}
    componentAppliedLinearAccClamp=<3000 3000 3000>
    componentLinearAccScale=<0.7 0.7 0.6>
    componentLinearVelScale=<0.8 0.8 0.7>
    gravityOverride {z=-980}
    physicsBodyDefinitions [{
      boundBone {boneName=dyn_necklace_01}
      boxExtents=<17 17 17>
      constraintSetup {angularLimitsMax=<0 0 25>, angularLimitsMin=<0 0 0>, angularTarget=<0 0 0>}
      localJointOffset=<1.5 0 0>
     }]
   }
   n84 = "Constraint - dyn_SkirtA00_l (PPPPPPPP)"!(
    ComponentPose=@n83
    Alpha=1
    ConstraintWeights_0=0.013
    ConstraintWeights_1:float=0.0426
    ConstraintWeights_2:float=0.296
    ConstraintWeights_3:float=0.3705
    ConstraintWeights_4:float=0.1231
    ConstraintWeights_5:float=0.0151
    ConstraintWeights_6:float=0.1155
    ConstraintWeights_7:float=0.0241
   ) -> (Pose) {
    alphaBoolBlend {blendOption=Linear}
    boneToModify {boneName=dyn_SkirtA00_l}
    constraintSetup [
     {targetBone {boneName=pelvis}}
     {targetBone {boneName=spine_01}}
     {targetBone {boneName=spine_02}}
     {targetBone {boneName=spine_03}}
     {targetBone {boneName=spine_04}}
     {targetBone {boneName=spine_05}}
     {targetBone {boneName=clavicle_l}}
     {targetBone {boneName=upperarm_twist_01_l}}
    ]
    constraintWeights [1,1,1,1,1,1,1,1]
   }
   n85 = "Constraint - dyn_SkirtB00_l (PPPPPPPP)"!(
    ComponentPose=@n84
    Alpha=1
    ConstraintWeights_0=0.0422
    ConstraintWeights_1:float=0.0656
    ConstraintWeights_2:float=0.1677
    ConstraintWeights_3:float=0.2635
    ConstraintWeights_4:float=0.0972
    ConstraintWeights_5:float=0.2375
    ConstraintWeights_6:float=0.1214
    ConstraintWeights_7:float=0.0049
   ) -> (Pose) {
    alphaBoolBlend {blendOption=Linear}
    boneToModify {boneName=dyn_SkirtB00_l}
    constraintSetup [
     {targetBone {boneName=pelvis}}
     {targetBone {boneName=spine_01}}
     {targetBone {boneName=spine_02}}
     {targetBone {boneName=spine_03}}
     {targetBone {boneName=spine_04}}
     {targetBone {boneName=clavicle_l}}
     {targetBone {boneName=upperarm_twist_01_l}}
     {targetBone {boneName=upperarm_twist_02_l}}
    ]
    constraintWeights [1,1,1,1,1,1,1,1]
   }
   n86 = "Constraint - dyn_SkirtC00_l (PPPPPPPP)"!(
    ComponentPose=@n85
    Alpha=1
    ConstraintWeights_0=0.0725
    ConstraintWeights_1:float=0.0852
    ConstraintWeights_2:float=0.1442
    ConstraintWeights_3:float=0.3092
    ConstraintWeights_4:float=0.111
    ConstraintWeights_5:float=0.0267
    ConstraintWeights_6:float=0.2109
    ConstraintWeights_7:float=0.0402
   ) -> (Pose) {
    alphaBoolBlend {blendOption=Linear}
    boneToModify {boneName=dyn_SkirtC00_l}
    constraintSetup [
     {targetBone {boneName=pelvis}}
     {targetBone {boneName=spine_01}}
     {targetBone {boneName=spine_02}}
     {targetBone {boneName=spine_03}}
     {targetBone {boneName=spine_04}}
     {targetBone {boneName=spine_05}}
     {targetBone {boneName=clavicle_l}}
     {targetBone {boneName=upperarm_twist_01_l}}
    ]
    constraintWeights [1,1,1,1,1,1,1,1]
   }
   n87 = "Constraint - dyn_SkirtC00_r (PPPPPPPP)"!(
    ComponentPose=@n86
    Alpha=1
    ConstraintWeights_0=0.0825
    ConstraintWeights_1:float=0.1533
    ConstraintWeights_2:float=0.3429
    ConstraintWeights_3:float=0.1282
    ConstraintWeights_4:float=0.0267
    ConstraintWeights_5:float=0.0728
    ConstraintWeights_6:float=0.1808
    ConstraintWeights_7:float=0.0128
   ) -> (Pose) {
    alphaBoolBlend {blendOption=Linear}
    boneToModify {boneName=dyn_SkirtC00_r}
    constraintSetup [
     {targetBone {boneName=spine_01}}
     {targetBone {boneName=spine_02}}
     {targetBone {boneName=spine_03}}
     {targetBone {boneName=spine_04}}
     {targetBone {boneName=spine_05}}
     {targetBone {boneName=clavicle_l}}
     {targetBone {boneName=clavicle_r}}
     {targetBone {boneName=upperarm_twist_01_r}}
    ]
    constraintWeights [1,1,1,1,1,1,1,1]
   }
   n88 = "Constraint - dyn_SkirtB00_r (PPPPPPPP)"!(
    ComponentPose=@n87
    Alpha=1
    ConstraintWeights_0=0.0305
    ConstraintWeights_1:float=0.056
    ConstraintWeights_2:float=0.1491
    ConstraintWeights_3:float=0.2849
    ConstraintWeights_4:float=0.105
    ConstraintWeights_5:float=0.2666
    ConstraintWeights_6:float=0.1013
    ConstraintWeights_7:float=0.0066
   ) -> (Pose) {
    alphaBoolBlend {blendOption=Linear}
    boneToModify {boneName=dyn_SkirtB00_r}
    constraintSetup [
     {targetBone {boneName=pelvis}}
     {targetBone {boneName=spine_01}}
     {targetBone {boneName=spine_02}}
     {targetBone {boneName=spine_03}}
     {targetBone {boneName=spine_04}}
     {targetBone {boneName=clavicle_r}}
     {targetBone {boneName=upperarm_twist_01_r}}
     {targetBone {boneName=upperarm_twist_02_r}}
    ]
    constraintWeights [1,1,1,1,1,1,1,1]
   }
   n89 = "Constraint - dyn_SkirtA00_r (PPPPPPPP)"!(
    ComponentPose=@n88
    Alpha=1
    ConstraintWeights_0=0.022
    ConstraintWeights_1:float=0.0444
    ConstraintWeights_2:float=0.2582
    ConstraintWeights_3:float=0.364
    ConstraintWeights_4:float=0.1341
    ConstraintWeights_5:float=0.0134
    ConstraintWeights_6:float=0.1407
    ConstraintWeights_7:float=0.0232
   ) -> (Pose) {
    alphaBoolBlend {blendOption=Linear}
    boneToModify {boneName=dyn_SkirtA00_r}
    constraintSetup [
     {targetBone {boneName=pelvis}}
     {targetBone {boneName=spine_01}}
     {targetBone {boneName=spine_02}}
     {targetBone {boneName=spine_03}}
     {targetBone {boneName=spine_04}}
     {targetBone {boneName=spine_05}}
     {targetBone {boneName=clavicle_r}}
     {targetBone {boneName=upperarm_twist_01_r}}
    ]
    constraintWeights [1,1,1,1,1,1,1,1]
   }
   n90 = "Transform (Modify) Bone - Bone: dyn_SkirtC00_l"(ComponentPose=@n89, Alpha=1, Translation="0,2,0", Rotation="0,0,0", Scale="1,1,1") {
    rotationMode=BMM_Additive
    rotationSpace=BCS_BoneSpace
    translationMode=BMM_Additive
    translationSpace=BCS_BoneSpace
    alphaBoolBlend {blendInTime=0.7, blendOption=Linear, blendOutTime=0.8}
    boneToModify {boneName=dyn_SkirtC00_l}
    translation {y=2}
   }
   n91 = "Transform (Modify) Bone - Bone: dyn_SkirtC00_r"(ComponentPose=@n90, Alpha=1, Translation="0,2,0", Rotation="0,0,0", Scale="1,1,1") {
    rotationMode=BMM_Additive
    rotationSpace=BCS_BoneSpace
    translationMode=BMM_Additive
    translationSpace=BCS_BoneSpace
    alphaBoolBlend {blendInTime=0.7, blendOption=Linear, blendOutTime=0.8}
    boneToModify {boneName=dyn_SkirtC00_r}
    translation {y=2}
   }
   n92 = "Transform (Modify) Bone - Bone: dyn_SkirtB00_l"(ComponentPose=@n91, Alpha=1, Translation="0,2,0", Rotation="0,0,0", Scale="1,1,1") {
    rotationMode=BMM_Additive
    rotationSpace=BCS_BoneSpace
    translationMode=BMM_Additive
    translationSpace=BCS_BoneSpace
    alphaBoolBlend {blendInTime=0.7, blendOption=Linear, blendOutTime=0.8}
    boneToModify {boneName=dyn_SkirtB00_l}
    translation {y=2}
   }
   n93 = "Transform (Modify) Bone - Bone: dyn_SkirtB00_r"(ComponentPose=@n92, Alpha=1, Translation="0,2,0", Rotation="0,0,0", Scale="1,1,1") {
    rotationMode=BMM_Additive
    rotationSpace=BCS_BoneSpace
    translationMode=BMM_Additive
    translationSpace=BCS_BoneSpace
    alphaBoolBlend {blendInTime=0.7, blendOption=Linear, blendOutTime=0.8}
    boneToModify {boneName=dyn_SkirtB00_r}
    translation {y=2}
   }
   n94 = "Constraint - shoulderpad_MSTR (RR)"!(ComponentPose=@n93, Alpha=1, ConstraintWeights_0=0.5, ConstraintWeights_1:float=0.5) -> (Pose) {
    alphaBoolBlend {blendOption=Linear}
    boneToModify {boneName=shoulderpad_MSTR}
    constraintSetup [{transformType=Rotation, targetBone {boneName=upperarm_twist_02_l}}, {transformType=Rotation, targetBone {boneName=upperarm_l}}]
    constraintWeights [1,1]
   }
   n95 = "Transform (Modify) Bone - Bone: dyn_shoulderpad_01_MSTR"!(ComponentPose=@n94, bAlphaBoolEnabled:bool=Get bIsSkydiveFloating(), Translation="1,0,1", Rotation="-15,0,0", Scale="1,1,1") -> (Pose) {
    alphaInputType=Bool
    rotationMode=BMM_Additive
    rotationSpace=BCS_BoneSpace
    scaleSpace=BCS_BoneSpace
    translationMode=BMM_Additive
    translationSpace=BCS_BoneSpace
    alphaBoolBlend {blendInTime=0.1, blendOption=Linear, blendOutTime=0.2}
    boneToModify {boneName=dyn_shoulderpad_01_MSTR}
    rotation {pitch=-15}
    translation {x=1, z=1}
   }
   n96 = Get bIsParachuteOpened()
   n97 = "Transform (Modify) Bone - Bone: dyn_shoulderpad_01_MSTR"(ComponentPose=@n95, bAlphaBoolEnabled=@n96, Translation="1,0,1", Rotation="-10,0,0", Scale="1,1,1") {
    alphaInputType=Bool
    rotationMode=BMM_Additive
    rotationSpace=BCS_BoneSpace
    scaleSpace=BCS_BoneSpace
    translationMode=BMM_Additive
    translationSpace=BCS_BoneSpace
    alphaBoolBlend {blendInTime=0.1, blendOption=Linear, blendOutTime=0.2}
    boneToModify {boneName=dyn_shoulderpad_01_MSTR}
    rotation {pitch=-10}
    translation {x=1, z=1}
   }
   n98 = "Transform (Modify) Bone - Bone: shoulderpad_MSTR"(ComponentPose=@n97, bAlphaBoolEnabled=@n96, Translation="5,0,0", Rotation="0,0,0", Scale="1,1,1") {
    alphaInputType=Bool
    rotationMode=BMM_Additive
    rotationSpace=BCS_BoneSpace
    translationMode=BMM_Additive
    translationSpace=BCS_BoneSpace
    alphaBoolBlend {blendInTime=0.1, blendOption=Linear, blendOutTime=0.2}
    boneToModify {boneName=shoulderpad_MSTR}
    translation {x=5}
   }
   n99 = "Constraint - dyn_zipper_r_MSTR (PPPPPPP)"(
    ComponentPose=@n98
    Alpha=1
    ConstraintWeights_0=0.0392
    ConstraintWeights_1=0.1758
    ConstraintWeights_2=0.2818
    ConstraintWeights_3=0.1208
    ConstraintWeights_4=0.2079
    ConstraintWeights_5=0.0739
    ConstraintWeights_6=0.1005
   ) {
    alphaBoolBlend {blendOption=Linear}
    boneToModify {boneName=dyn_zipper_r_MSTR}
    constraintSetup [
     {targetBone {boneName=spine_01}}
     {targetBone {boneName=spine_02}}
     {targetBone {boneName=spine_03}}
     {targetBone {boneName=spine_04}}
     {targetBone {boneName=clavicle_r}}
     {targetBone {boneName=upperarm_twist_01_r}}
     {targetBone {boneName=dyn_SkirtA01_r}}
    ]
    constraintWeights [1,1,1,1,1,1,1]
   }
   n100 = "Constraint - dyn_zipper_l_MSTR (PPPPPPPP)"(
    ComponentPose=@n99
    Alpha=1
    ConstraintWeights_0=0.0138
    ConstraintWeights_1=0.0334
    ConstraintWeights_2=0.1692
    ConstraintWeights_3=0.2506
    ConstraintWeights_4=0.1065
    ConstraintWeights_5=0.1837
    ConstraintWeights_6=0.0781
    ConstraintWeights_7=0.1647
   ) {
    alphaBoolBlend {blendOption=Linear}
    boneToModify {boneName=dyn_zipper_l_MSTR}
    constraintSetup [
     {targetBone {boneName=pelvis}}
     {targetBone {boneName=spine_01}}
     {targetBone {boneName=spine_02}}
     {targetBone {boneName=spine_03}}
     {targetBone {boneName=spine_04}}
     {targetBone {boneName=clavicle_l}}
     {targetBone {boneName=upperarm_twist_01_l}}
     {targetBone {boneName=dyn_SkirtA01_l}}
    ]
    constraintWeights [1,1,1,1,1,1,1,1]
   }
   Output Pose(Result=Component To Local(ComponentPose=@n100)) {layerGroup=None}
   n101 = Get Right_leg_belt_crouch() -> (Right_leg_belt_crouch:double)
   n102 = Control Rig!(
    Source=Copy Pose From Mesh() {bUseAttachedParent=true}
    Alpha
    is SkyDive Floating:double=Get Is Sky Dive Floating() -> (Is Sky Dive Floating:double)
    is Playing Emote:double=Get Is Playing Emote() -> (Is Playing Emote:double)
    Is Moving 2D:double=Get Is Moving 2D() -> (Is Moving 2D:double)
    Is Parachute Opened:double=Get Is Parachute Opened() -> (Is Parachute Opened:double)
    Is Siding:double=Get Is Siding() -> (Is Siding:double)
    Is Crouching:double=Get Is Crouching() -> (Is Crouching:double)
    Is Playing Melee Anim:double=Get Is Playing Melee Anim() -> (Is Playing Melee Anim:double)
   ) -> (Pose) {
    alphaBoolBlend {blendOption=Linear}
    controlRigAssetReference {blueprintRigClass=&/M_MED_QuickBurst_Plains_CtrlRig.M_MED_QuickBurst_Plains_CtrlRig_C}
   }
   n103 = "Transform (Modify) Bone - Bone: dyn_kneepad_l"(
    ComponentPose=Local To Component(LocalPose=@n102)
    Alpha=Get calf_l_front() -> (calf_l_front:double)
    Translation="0,-2.5,0"
    Rotation="0,0,0"
    Scale="1,1,1"
   ) {translationMode=BMM_Additive, alphaBoolBlend {blendOption=Linear}, boneToModify {boneName=dyn_kneepad_l}, translation {y=-2.5}}
   n104 = "Constraint - dyn_kneepad_l (RR)"!(ComponentPose=@n103, Alpha=1, ConstraintWeights_0=0.5, ConstraintWeights_1:float=0.5) -> (Pose) {
    alphaBoolBlend {blendOption=Linear}
    boneToModify {boneName=dyn_kneepad_l}
    constraintSetup [{transformType=Rotation, targetBone {boneName=thigh_l}}, {transformType=Rotation, targetBone {boneName=calf_l}}]
    constraintWeights [1,1]
   }
   n105 = "Transform (Modify) Bone - Bone: dyn_kneepad_r"(ComponentPose=@n104, Alpha=Get calf_r_front() -> (calf_r_front:double), Translation="2,2,-0.5", Rotation="-5,-70,-5", Scale="1,1,1") {
    rotationMode=BMM_Additive
    rotationSpace=BCS_BoneSpace
    translationMode=BMM_Additive
    translationSpace=BCS_BoneSpace
    alphaBoolBlend {blendOption=Linear}
    boneToModify {boneName=dyn_kneepad_r}
    rotation=<rot -5 -70 -5>
    translation=<2 2 -0.5>
   }
   n106 = "Constraint - dyn_collar_MSTR (RR)"(ComponentPose=@n105, bAlphaBoolEnabled=Get Helmet_on() -> (Helmet_on:bool), ConstraintWeights_0=0.55, ConstraintWeights_1=0.45) {
    alphaInputType=Bool
    alphaBoolBlend {blendOption=Linear}
    boneToModify {boneName=dyn_collar_MSTR}
    constraintSetup [{transformType=Rotation, targetBone {boneName=head}}, {transformType=Rotation, targetBone {boneName=spine_05}}]
    constraintWeights [1,1]
   }
   n107 = "Transform (Modify) Bone - Bone: dyn_collar_01"(ComponentPose=@n106, Alpha=@n9, Translation="0,-5,0", Rotation="0,-90,0", Scale="1,1,1") {
    rotationMode=BMM_Additive
    rotationSpace=BCS_BoneSpace
    scaleMode=BMM_Additive
    scaleSpace=BCS_BoneSpace
    translationMode=BMM_Additive
    translationSpace=BCS_BoneSpace
    alphaBoolBlend {blendOption=Linear}
    boneToModify {boneName=dyn_collar_01}
    rotation {yaw=-90}
    translation {y=-5}
   }
   n108 = "Transform (Modify) Bone - Bone: dyn_collarB_01_r"!(ComponentPose=@n107, Alpha:float=@n9, Translation="0,0,0", Rotation="20,20,0", Scale="1,1,1") -> (Pose) {
    rotationMode=BMM_Additive
    rotationSpace=BCS_BoneSpace
    scaleMode=BMM_Additive
    scaleSpace=BCS_BoneSpace
    translationMode=BMM_Additive
    translationSpace=BCS_BoneSpace
    alphaBoolBlend {blendOption=Linear}
    boneToModify {boneName=dyn_collarB_01_r}
    rotation {pitch=20, yaw=20}
   }
   n109 = "Transform (Modify) Bone - Bone: dyn_collarA_01_r"!(ComponentPose=@n108, Alpha:float=@n9, Translation="0,0,0", Rotation="0,30,0", Scale="1,1,1") -> (Pose) {
    rotationMode=BMM_Additive
    rotationSpace=BCS_BoneSpace
    scaleMode=BMM_Additive
    scaleSpace=BCS_BoneSpace
    translationMode=BMM_Additive
    translationSpace=BCS_BoneSpace
    alphaBoolBlend {blendOption=Linear}
    boneToModify {boneName=dyn_collarA_01_r}
    rotation {yaw=30}
   }
   n110 = "Transform (Modify) Bone - Bone: dyn_collarB_01_r"(ComponentPose=@n109, Alpha=Get neck_rot_r() -> (neck_rot_r:double), Translation="0,0,0", Rotation="50,10,-40", Scale="1,1,1") {
    rotationMode=BMM_Additive
    rotationSpace=BCS_BoneSpace
    scaleMode=BMM_Additive
    scaleSpace=BCS_BoneSpace
    translationMode=BMM_Additive
    translationSpace=BCS_BoneSpace
    alphaBoolBlend {blendOption=Linear}
    boneToModify {boneName=dyn_collarB_01_r}
    rotation=<rot 50 10 -40>
   }
   n111 = "Transform (Modify) Bone - Bone: dyn_collarA_01_l"(ComponentPose=@n110, Alpha=Get neck_rot_l() -> (neck_rot_l:double), Translation="0,-2,0", Rotation="0,-30,20", Scale="1,1,1") {
    rotationMode=BMM_Additive
    rotationSpace=BCS_BoneSpace
    scaleMode=BMM_Additive
    scaleSpace=BCS_BoneSpace
    translationMode=BMM_Additive
    translationSpace=BCS_BoneSpace
    alphaBoolBlend {blendOption=Linear}
    boneToModify {boneName=dyn_collarA_01_l}
    rotation {roll=20, yaw=-30}
    translation {y=-2}
   }
   n112 = "Transform (Modify) Bone - Bone: dyn_collarA_01_l"(ComponentPose=@n111, Alpha=Get Shoulder_rot_f_l() -> (Shoulder_rot_f_l:double), Translation="0,-2,0", Rotation="0,0,0", Scale="1,1,1") {
    rotationMode=BMM_Additive
    rotationSpace=BCS_BoneSpace
    scaleMode=BMM_Additive
    scaleSpace=BCS_BoneSpace
    translationMode=BMM_Additive
    translationSpace=BCS_BoneSpace
    alphaBoolBlend {blendOption=Linear}
    boneToModify {boneName=dyn_collarA_01_l}
    translation {y=-2}
   }
   n113 = "Transform (Modify) Bone - Bone: dyn_collarA_01_l"(ComponentPose=@n112, Alpha=1, Translation="2,0,0", Rotation="0,0,0", Scale="1,1,1") {
    rotationMode=BMM_Additive
    rotationSpace=BCS_BoneSpace
    scaleMode=BMM_Additive
    scaleSpace=BCS_BoneSpace
    translationMode=BMM_Additive
    translationSpace=BCS_BoneSpace
    alphaBoolBlend {blendOption=Linear}
    boneToModify {boneName=dyn_collarA_01_l}
    translation {x=2}
   }
   n114 = "Transform (Modify) Bone - Bone: dyn_collarB_01_r"(ComponentPose=@n113, Alpha=@n9, Translation="0,0,0", Rotation="48.040,2.133,4.360", Scale="1,1,1") {
    rotationMode=BMM_Additive
    rotationSpace=BCS_BoneSpace
    scaleMode=BMM_Additive
    scaleSpace=BCS_BoneSpace
    translationMode=BMM_Additive
    translationSpace=BCS_BoneSpace
    alphaBoolBlend {blendOption=Linear}
    boneToModify {boneName=dyn_collarB_01_r}
    rotation=<rot 48.04 2.133 4.36>
   }
   n115 = "Transform (Modify) Bone - Bone: dyn_collarB_01_l"(ComponentPose=@n114, Alpha=@n9, Translation="0,0,0", Rotation="11.446,0.525,0.349", Scale="1,1,1") {
    rotationMode=BMM_Additive
    rotationSpace=BCS_BoneSpace
    scaleMode=BMM_Additive
    scaleSpace=BCS_BoneSpace
    translationMode=BMM_Additive
    translationSpace=BCS_BoneSpace
    alphaBoolBlend {blendOption=Linear}
    boneToModify {boneName=dyn_collarB_01_l}
    rotation=<rot 11.446 0.525 0.349>
   }
   n116 = Anim Dynamics - Bone: dyn_collarA_01_l(ComponentPose=@n115, Alpha=0.5, ExternalForce="0,0,0") {
    angularDampingOverride=0.999
    angularSpringConstant=800
    bAngularSpring=true
    bGravityOverrideInSimSpace=true
    bOverrideAngularDamping=true
    bUseGravityOverride=true
    gravityScale=0.2
    simulationSpace=RootRelative
    alphaBoolBlend {blendOption=Linear}
    boundBone {boneName=dyn_collarA_01_l}
    componentAppliedLinearAccClamp=<1500 1500 1500>
    componentLinearAccScale=<0.5 0.5 1>
    componentLinearVelScale=<0.5 0.5 1>
    gravityOverride {z=-300}
    physicsBodyDefinitions [{
      sphereCollisionRadius=0
      boundBone {boneName=dyn_collarA_01_l}
      boxExtents=<15 5 5>
      constraintSetup {angularTargetAxis=AxisY, angularLimitsMax=<45 45 25>, angularLimitsMin=<-45 -45 -25>, angularTarget=<0 0 0>}
      localJointOffset=<5 0 0>
     }]
    relativeSpaceBone {boneName=spine_05}
   }
   n117 = Anim Dynamics - Bone: dyn_collarB_01_l(ComponentPose=@n116, Alpha=0.5, ExternalForce="0,0,0") {
    angularDampingOverride=0.999
    angularSpringConstant=800
    bAngularSpring=true
    bGravityOverrideInSimSpace=true
    bOverrideAngularDamping=true
    bUseGravityOverride=true
    gravityScale=0.2
    simulationSpace=RootRelative
    alphaBoolBlend {blendOption=Linear}
    boundBone {boneName=dyn_collarB_01_l}
    componentAppliedLinearAccClamp=<500 500 500>
    componentLinearAccScale=<0.5 0.5 0.1>
    componentLinearVelScale=<0.5 0.5 0.1>
    gravityOverride {z=-300}
    physicsBodyDefinitions [{
      sphereCollisionRadius=0
      boundBone {boneName=dyn_collarB_01_l}
      boxExtents=<15 15 15>
      constraintSetup {angularTargetAxis=AxisY, angularLimitsMax=<1 1 15>, angularLimitsMin=<-1 -1 -15>, angularTarget=<0 0 0>}
      localJointOffset=<-5 0.1 -1>
     }]
    relativeSpaceBone {boneName=spine_05}
   }
   n118 = Anim Dynamics - Bone: dyn_collar_01(ComponentPose=@n117, Alpha=0.5, ExternalForce="0,0,0") {
    angularDampingOverride=0.999
    angularSpringConstant=800
    bAngularSpring=true
    bGravityOverrideInSimSpace=true
    bOverrideAngularDamping=true
    bUseGravityOverride=true
    gravityScale=0.2
    simulationSpace=RootRelative
    alphaBoolBlend {blendOption=Linear}
    boundBone {boneName=dyn_collar_01}
    componentAppliedLinearAccClamp=<500 500 500>
    componentLinearAccScale=<0.5 0.5 0.1>
    componentLinearVelScale=<0.5 0.5 0.1>
    gravityOverride {z=-300}
    physicsBodyDefinitions [{
      sphereCollisionRadius=0
      boundBone {boneName=dyn_collar_01}
      boxExtents=<25 25 25>
      constraintSetup {angularTargetAxis=AxisY, angularLimitsMax=<1 5 25>, angularLimitsMin=<-1 -5 -25>, angularTarget=<0 0 0>}
      localJointOffset=<-5 -1 -0.1>
     }]
    relativeSpaceBone {boneName=spine_05}
   }
   n119 = Anim Dynamics - Bone: dyn_collarA_01_r(ComponentPose=@n118, Alpha=1, ExternalForce="0,0,0") {
    angularDampingOverride=0.999
    angularSpringConstant=800
    bAngularSpring=true
    bGravityOverrideInSimSpace=true
    bOverrideAngularDamping=true
    bUseGravityOverride=true
    gravityScale=0.2
    simulationSpace=RootRelative
    alphaBoolBlend {blendOption=Linear}
    boundBone {boneName=dyn_collarA_01_r}
    componentAppliedLinearAccClamp=<1500 1500 1500>
    componentLinearAccScale=<0.5 0.5 1>
    componentLinearVelScale=<0.5 0.5 1>
    gravityOverride {z=-300}
    physicsBodyDefinitions [{
      sphereCollisionRadius=0
      boundBone {boneName=dyn_collarA_01_r}
      boxExtents=<15 5 15>
      constraintSetup {angularTargetAxis=AxisY, angularLimitsMax=<45 1 1>, angularLimitsMin=<20 -15 -1>, angularTarget=<0 0 0>}
      localJointOffset=<5 0 0>
     }]
    relativeSpaceBone {boneName=spine_05}
   }
   n120 = Anim Dynamics - Bone: dyn_collarB_01_r(ComponentPose=@n119, Alpha=1, ExternalForce="0,0,0") {
    angularDampingOverride=0.999
    angularSpringConstant=800
    bAngularSpring=true
    bGravityOverrideInSimSpace=true
    bOverrideAngularDamping=true
    bUseGravityOverride=true
    gravityScale=0.2
    simulationSpace=RootRelative
    alphaBoolBlend {blendOption=Linear}
    boundBone {boneName=dyn_collarB_01_r}
    componentAppliedLinearAccClamp=<500 500 500>
    componentLinearAccScale=<0.5 0.5 0.1>
    componentLinearVelScale=<0.5 0.5 0.1>
    gravityOverride {z=-300}
    physicsBodyDefinitions [{
      sphereCollisionRadius=0
      boundBone {boneName=dyn_collarB_01_r}
      boxExtents=<15 5 15>
      constraintSetup {angularTargetAxis=AxisY, angularLimitsMax=<45 1 15>, angularLimitsMin=<20 -15 -15>, angularTarget=<0 0 0>}
      localJointOffset=<5 0 0>
     }]
    relativeSpaceBone {boneName=spine_05}
   }
   n121 = Get bIsJumping()
   n122 = "Transform (Modify) Bone - Bone: dyn_collarA_01_r"!(ComponentPose=@n120, bAlphaBoolEnabled:bool=@n121, Translation="0,0,0", Rotation="0,-15,25", Scale="1,1,1") -> (Pose) {
    alphaInputType=Bool
    rotationMode=BMM_Additive
    rotationSpace=BCS_BoneSpace
    translationMode=BMM_Additive
    translationSpace=BCS_BoneSpace
    alphaBoolBlend {blendInTime=0.3, blendOption=Custom, blendOutTime=0.8}
    boneToModify {boneName=dyn_collarA_01_r}
    rotation {roll=25, yaw=-15}
   }
   n123 = "Transform (Modify) Bone - Bone: dyn_collarA_02_r"(ComponentPose=@n122, bAlphaBoolEnabled=@n121, Translation="0,0,0", Rotation="-15,0,15", Scale="1,1,1") {
    alphaInputType=Bool
    rotationMode=BMM_Additive
    rotationSpace=BCS_BoneSpace
    translationMode=BMM_Additive
    translationSpace=BCS_BoneSpace
    alphaBoolBlend {blendInTime=0.3, blendOption=Custom, blendOutTime=0.8}
    boneToModify {boneName=dyn_collarA_02_r}
    rotation {pitch=-15, roll=15}
   }
   n124 = "Transform (Modify) Bone - Bone: dyn_collarB_01_r"!(ComponentPose=@n123, bAlphaBoolEnabled:bool=@n121, Translation="0,0,0", Rotation="0,0,25", Scale="1,1,1") -> (Pose) {
    alphaInputType=Bool
    rotationMode=BMM_Additive
    rotationSpace=BCS_BoneSpace
    translationMode=BMM_Additive
    translationSpace=BCS_BoneSpace
    alphaBoolBlend {blendInTime=0.3, blendOption=Custom, blendOutTime=0.8}
    boneToModify {boneName=dyn_collarB_01_r}
    rotation {roll=25}
   }
   n125 = "Transform (Modify) Bone - Bone: dyn_belt_base"(ComponentPose=@n124, Alpha=@n101, Translation="0,0,0", Rotation="-25,0,-25", Scale="1,1,1") {
    rotationMode=BMM_Additive
    rotationSpace=BCS_BoneSpace
    translationMode=BMM_Additive
    translationSpace=BCS_BoneSpace
    alphaBoolBlend {blendOption=Linear}
    boneToModify {boneName=dyn_belt_base}
    rotation {pitch=-25, roll=-25}
   }
   n126 = "Transform (Modify) Bone - Bone: dyn_belt_center"!(ComponentPose=@n125, Alpha:float=@n101, Translation="0,3,0", Rotation="10,0,0", Scale="1,1,1") -> (Pose) {
    rotationMode=BMM_Additive
    rotationSpace=BCS_BoneSpace
    translationMode=BMM_Additive
    translationSpace=BCS_BoneSpace
    alphaBoolBlend {blendInTime=0.7, blendOption=Linear, blendOutTime=0.8}
    boneToModify {boneName=dyn_belt_center}
    rotation {pitch=10}
    translation {y=3}
   }
   n127 = "Transform (Modify) Bone - Bone: dyn_belt_rotator_front"(ComponentPose=@n126, Alpha=Get Left_leg_belt_crouch() -> (Left_leg_belt_crouch:float), Translation="0,0,1", Rotation="15,0,0", Scale="1,1,1") {
    rotationMode=BMM_Additive
    rotationSpace=BCS_BoneSpace
    translationMode=BMM_Additive
    translationSpace=BCS_BoneSpace
    alphaBoolBlend {blendInTime=0.7, blendOption=Linear, blendOutTime=0.8}
    boneToModify {boneName=dyn_belt_rotator_front}
    rotation {pitch=15}
    translation {z=1}
   }
   n128 = "Transform (Modify) Bone - Bone: dyn_belt"!(
    ComponentPose=@n127
    Alpha:float=Get Right_leg_jump_side() -> (Right_leg_jump_side:float)
    Translation="0,0,4"
    Rotation="-25,0,0"
    Scale="1,1,1"
   ) -> (Pose) {
    rotationMode=BMM_Additive
    translationMode=BMM_Additive
    alphaBoolBlend {blendInTime=0.7, blendOption=Linear, blendOutTime=0.8}
    boneToModify {boneName=dyn_belt}
    rotation {pitch=-25}
    translation {z=4}
   }
   n129 = "Constraint - dyn_belt_rotator_MSTR (PPPP)"!(
    ComponentPose=@n128
    Alpha=1
    ConstraintWeights_0=0.7104
    ConstraintWeights_1:float=0.2707
    ConstraintWeights_2:float=0.015
    ConstraintWeights_3:float=0.0038
   ) -> (Pose) {
    alphaBoolBlend {blendOption=Linear}
    boneToModify {boneName=dyn_belt_rotator_MSTR}
    constraintSetup [{targetBone {boneName=pelvis}}, {targetBone {boneName=spine_01}}, {targetBone {boneName=spine_02}}, {targetBone {boneName=spine_03}}]
    constraintWeights [1,1,1,1]
   }
   n130 = "Constraint - dyn_belt_rotator_front_MSTR (PPPPP)"!(
    ComponentPose=@n129
    Alpha=1
    ConstraintWeights_0=0.4512
    ConstraintWeights_1:float=0.4272
    ConstraintWeights_2:float=0.0879
    ConstraintWeights_3:float=0.0331
    ConstraintWeights_4:float=0.0005
   ) -> (Pose) {
    alphaBoolBlend {blendOption=Linear}
    boneToModify {boneName=dyn_belt_rotator_front_MSTR}
    constraintSetup [
     {targetBone {boneName=pelvis}}
     {targetBone {boneName=spine_01}}
     {targetBone {boneName=spine_02}}
     {targetBone {boneName=spine_03}}
     {targetBone {boneName=thigh_twist_01_l}}
    ]
    constraintWeights [1,1,1,1,1]
   }
   n131 = "Transform (Modify) Bone - Bone: dyn_belt_rotator"(ComponentPose=@n130, Alpha=Get Right_leg_jump_side(), Translation="0,0,1", Rotation="-25,0,0", Scale="1,1,1") {
    rotationMode=BMM_Additive
    rotationSpace=BCS_BoneSpace
    translationMode=BMM_Additive
    translationSpace=BCS_BoneSpace
    alphaBoolBlend {blendInTime=0.7, blendOption=Linear, blendOutTime=0.8}
    boneToModify {boneName=dyn_belt_rotator}
    rotation {pitch=-25}
    translation {z=1}
   }
   n132 = "Constraint - dyn_belt_bags_MSTR (PPPP)"!(
    ComponentPose=@n131
    bAlphaBoolEnabled:bool=Get Not Crouching()
    ConstraintWeights_0=0.4171
    ConstraintWeights_1:float=0.4136
    ConstraintWeights_2:float=0.1233
    ConstraintWeights_3:float=0.046
   ) -> (Pose) {
    alphaInputType=Bool
    alphaBoolBlend {blendInTime=0.1, blendOption=Linear, blendOutTime=0.1}
    boneToModify {boneName=dyn_belt_bags_MSTR}
    constraintSetup [{targetBone {boneName=pelvis}}, {targetBone {boneName=spine_01}}, {targetBone {boneName=spine_02}}, {targetBone {boneName=spine_03}}]
    constraintWeights [1,1,1,1]
   }
   n133 = Get bIsCrouching()
   n134 = "Constraint - dyn_belt_bags_MSTR (P)"!(ComponentPose=@n132, bAlphaBoolEnabled:bool=@n133, ConstraintWeights_0=1) -> (Pose) {
    alphaInputType=Bool
    alphaBoolBlend {blendInTime=0.1, blendOption=Linear, blendOutTime=0.2}
    boneToModify {boneName=dyn_belt_bags_MSTR}
    constraintSetup [{offsetOption=None, targetBone {boneName=dyn_bags_const_AIM}}]
    constraintWeights [1]
   }
   n135 = "Constraint - dyn_bags_const_AIM (TTTT)"!(
    ComponentPose=@n134
    Alpha=1
    ConstraintWeights_0=0.447
    ConstraintWeights_1:float=0.3867
    ConstraintWeights_2:float=0.1231
    ConstraintWeights_3:float=0.0432
   ) -> (Pose) {
    alphaBoolBlend {blendInTime=0.1, blendOption=Linear, blendOutTime=0.2}
    boneToModify {boneName=dyn_bags_const_AIM}
    constraintSetup [
     {offsetOption=None, transformType=Translation, targetBone {boneName=pelvis}}
     {offsetOption=None, transformType=Translation, targetBone {boneName=spine_01}}
     {offsetOption=None, transformType=Translation, targetBone {boneName=spine_02}}
     {offsetOption=None, transformType=Translation, targetBone {boneName=spine_03}}
    ]
    constraintWeights [1,1,1,1]
   }
   n136 = "Transform (Modify) Bone - Bone: dyn_belt_bags_MSTR"!(ComponentPose=@n135, bAlphaBoolEnabled:bool=@n133, Translation="0,0,1", Rotation="0,0,10", Scale="1,1,1") -> (Pose) {
    alphaInputType=Bool
    rotationMode=BMM_Additive
    translationMode=BMM_Additive
    alphaBoolBlend {blendInTime=0.7, blendOption=Linear, blendOutTime=0.8}
    boneToModify {boneName=dyn_belt_bags_MSTR}
    rotation {roll=10}
    translation {z=1}
   }
   n137 = Get bIsPlayingEmote()
   n138 = "Transform (Modify) Bone - Bone: dyn_belt_bags_01_MSTR"(ComponentPose=@n136, bAlphaBoolEnabled=@n137, Translation="0,0,0", Rotation="15,-15,15", Scale="1,1,1") {
    alphaInputType=Bool
    rotationMode=BMM_Additive
    translationMode=BMM_Additive
    alphaBoolBlend {blendInTime=0.1, blendOption=Linear, blendOutTime=0.3}
    boneToModify {boneName=dyn_belt_bags_01_MSTR}
    rotation=<rot 15 -15 15>
   }
   n139 = "Transform (Modify) Bone - Bone: dyn_belt_bags_02_MSTR"(ComponentPose=@n138, bAlphaBoolEnabled=@n137, Translation="0,0,0", Rotation="15,-15,15", Scale="1,1,1") {
    alphaInputType=Bool
    rotationMode=BMM_Additive
    translationMode=BMM_Additive
    alphaBoolBlend {blendInTime=0.1, blendOption=Linear, blendOutTime=0.3}
    boneToModify {boneName=dyn_belt_bags_02_MSTR}
    rotation=<rot 15 -15 15>
   }
   n140 = Get Not Crouching and jumping() -> (Not Crouching and jumping:bool)
   n141 = Anim Dynamics - Bone: dyn_belt_bags_01!(ComponentPose=@n139, bAlphaBoolEnabled:bool=@n140, ExternalForce="0,0,0") -> (Pose) {
    alphaInputType=Bool
    angularDampingOverride=1
    bOverrideAngularDamping=true
    bUseGravityOverride=true
    bUsePlanarLimit=false
    bUseSphericalLimits=true
    simulationSpace=RootRelative
    alphaBoolBlend {blendInTime=0.1, blendOption=Linear, blendOutTime=0.2}
    boundBone {boneName=dyn_belt_bags_01}
    componentAppliedLinearAccClamp=<500 500 500>
    componentLinearAccScale=<0.1 0.1 0.1>
    componentLinearVelScale=<0.3 0.3 0.3>
    gravityOverride {z=-980}
    physicsBodyDefinitions [{
      boundBone {boneName=dyn_belt_bags_01}
      boxExtents=<6 10 20>
      constraintSetup {angularLimitsMax=<0 0 3>, angularLimitsMin=<0 -20 -3>, angularTarget=<0 0 0>}
      localJointOffset=<1 0 0>
     }]
    sphericalLimits [{limitRadius=18, drivingBone {boneName=thigh_twist_01_l}, sphereLocalOffset=<15 0 0>}]
   }
   n142 = Anim Dynamics - Bone: dyn_belt_bags_02!(ComponentPose=@n141, bAlphaBoolEnabled:bool=@n140, ExternalForce="0,0,0") -> (Pose) {
    alphaInputType=Bool
    angularDampingOverride=1
    bOverrideAngularDamping=true
    bUseGravityOverride=true
    bUsePlanarLimit=false
    bUseSphericalLimits=true
    simulationSpace=RootRelative
    alphaBoolBlend {blendInTime=0.1, blendOption=Linear, blendOutTime=0.2}
    boundBone {boneName=dyn_belt_bags_02}
    componentAppliedLinearAccClamp=<500 500 500>
    componentLinearAccScale=<0.1 0.1 0.1>
    componentLinearVelScale=<0.3 0.3 0.3>
    gravityOverride {z=-980}
    physicsBodyDefinitions [{
      boundBone {boneName=dyn_belt_bags_02}
      boxExtents=<6 10 20>
      constraintSetup {angularLimitsMax=<0 0 3>, angularLimitsMin=<0 -15 -3>, angularTarget=<0 0 0>}
      localJointOffset=<2 0 0>
     }]
    sphericalLimits [{limitRadius=18, drivingBone {boneName=thigh_twist_01_l}, sphereLocalOffset=<15 0 0>}]
   }
   n143 = "Transform (Modify) Bone - Bone: dyn_belt_bags_01_MSTR"(ComponentPose=@n142, bAlphaBoolEnabled=Get Not Crouching(), Translation="0,0,0", Rotation="1,-5,-2", Scale="1,1,1") {
    alphaInputType=Bool
    rotationMode=BMM_Additive
    translationMode=BMM_Additive
    alphaBoolBlend {blendInTime=0.1, blendOption=Custom, blendOutTime=0.2}
    boneToModify {boneName=dyn_belt_bags_01_MSTR}
    rotation=<rot 1 -5 -2>
   }
   n144 = Anim Dynamics - Bone: dyn_belt_bags_01(ComponentPose=@n143, bAlphaBoolEnabled=True, ExternalForce="0,0,0") {
    alphaInputType=Bool
    angularDampingOverride=1
    bOverrideAngularDamping=true
    bUseGravityOverride=true
    bUsePlanarLimit=false
    simulationSpace=RootRelative
    alphaBoolBlend {blendInTime=0.3, blendOption=Linear, blendOutTime=0.4}
    boundBone {boneName=dyn_belt_bags_01}
    componentAppliedLinearAccClamp=<500 500 500>
    componentLinearAccScale=<0.1 0.1 0.1>
    componentLinearVelScale=<0.3 0.3 0.3>
    gravityOverride {z=-980}
    physicsBodyDefinitions [{
      boundBone {boneName=dyn_belt_bags_01}
      boxExtents=<6 10 20>
      constraintSetup {angularLimitsMax=<0 0 3>, angularLimitsMin=<0 -20 -3>, angularTarget=<0 0 0>}
      localJointOffset=<2 0 0>
     }]
    sphericalLimits [{limitRadius=18, drivingBone {boneName=thigh_twist_01_l}, sphereLocalOffset=<15 0 0>}]
   }
   n145 = Anim Dynamics - Bone: dyn_belt_bags_02(ComponentPose=@n144, bAlphaBoolEnabled=True, ExternalForce="0,0,0") {
    alphaInputType=Bool
    angularDampingOverride=1
    bOverrideAngularDamping=true
    bUseGravityOverride=true
    bUsePlanarLimit=false
    simulationSpace=RootRelative
    alphaBoolBlend {blendInTime=0.3, blendOption=Linear, blendOutTime=0.4}
    boundBone {boneName=dyn_belt_bags_02}
    componentAppliedLinearAccClamp=<500 500 500>
    componentLinearAccScale=<0.1 0.1 0.1>
    componentLinearVelScale=<0.3 0.3 0.3>
    gravityOverride {z=-980}
    physicsBodyDefinitions [{
      boundBone {boneName=dyn_belt_bags_02}
      boxExtents=<6 10 20>
      constraintSetup {angularLimitsMax=<0 0 3>, angularLimitsMin=<0 -15 -3>, angularTarget=<0 0 0>}
      localJointOffset=<2 0 0>
     }]
    sphericalLimits [{limitRadius=18, drivingBone {boneName=thigh_twist_01_l}, sphereLocalOffset=<15 0 0>}]
   }
   n146 = "Transform (Modify) Bone - Bone: dyn_belt_bags_01_MSTR"(ComponentPose=@n145, bAlphaBoolEnabled=Get bIsJumping(), Translation="0,0,0", Rotation="25,-5,-25", Scale="1,1,1") {
    alphaInputType=Bool
    rotationMode=BMM_Additive
    translationMode=BMM_Additive
    alphaBoolBlend {blendInTime=0.3, blendOption=Custom, blendOutTime=0.8}
    boneToModify {boneName=dyn_belt_bags_01_MSTR}
    rotation=<rot 25 -5 -25>
   }
   Save cached pose 'Presim'(Pose=Component To Local(ComponentPose=@n146)) {cacheName=Presim}
   Reroute Node(InputPin:Wildcard) -> (OutputPin:Wildcard)
   Reroute Node(InputPin:float=Get Left_leg_belt_crouch()) -> (OutputPin:float)
   Get Left_leg_belt_crouch()
   Get bIsJumping()
  ]}
 EventGraph {nodes [
   n1 = Event OnCharacterCustomizationLoaded()
   n2 = GetSkinnedAsset!(self:Skinned Mesh Component Object Reference=Get Array Item with EFort Custom Part Type(Array=@n1.CustomMeshParts, PartType=Face)) -> (ReturnValue:Skinned Asset Object Reference)
   > Set Helmet_on(Helmet_on:bool="Equal Exactly (String)"(A=GetDisplayName(Object=@n2), B=Get isHelmet() -> (isHelmet:String))) -> (Output_Get:bool)
   n3 = Event BlueprintUpdateAnimation()
   > Set Delta Time X(Delta Time X=@n3.DeltaTimeX)
   > Set Shoulder_rot_f_l(Shoulder_rot_f_l:double=GetAlphaFromReferenceTransformDelta(
     SkeletalMeshComponent=Get Shoulder_Mesh() -> (Shoulder_Mesh:Skeletal Mesh Component Object Reference)
     BoneName=upperarm_l
     BaseName=clavicle_l
     InputRange=(X=0,Y=110)
     TwistAxis="0,0,1"
    )) -> (Output_Get:double)
   > Set calf_l_front(calf_l_front:double=GetAlphaFromReferenceTransformDelta(SkeletalMeshComponent=Get Skel_Mesh(), BoneName=thigh_l, BaseName=calf_l, InputRange=(X=0,Y=-90), TwistAxis="0,0,1")) -> (Output_Get:double)
   > Set calf_r_front(calf_r_front:double=GetAlphaFromReferenceTransformDelta(SkeletalMeshComponent=Get Skel_Mesh(), BoneName=thigh_r, BaseName=calf_r, InputRange=(X=-30,Y=-90), TwistAxis="0,0,1")) -> (Output_Get:double)
   > Set neck_rot_up(neck_rot_up:double=GetAlphaFromReferenceTransformDelta(
     SkeletalMeshComponent=Get Head_Mesh() -> (Head_Mesh:Skeletal Mesh Component Object Reference)
     BoneName=head
     BaseName=spine_05
     InputRange=(X=-20,Y=-80)
     TwistAxis="0,0,1"
    )) -> (Output_Get:double)
   > Set Neck Rot Up Helmet(Neck Rot Up Helmet:float=GetAlphaFromReferenceTransformDelta(SkeletalMeshComponent=Get Head_Mesh(), BoneName=head, BaseName=spine_05, InputRange=(X=30,Y=-30), TwistAxis="0,0,1")) -> (Output_Get:float)
   > Set neck_rot_r(neck_rot_r:double=GetAlphaFromReferenceTransformDelta(
     SkeletalMeshComponent=Get Head_Mesh()
     BoneName=head
     BaseName=spine_05
     DeltaComponentType=EulerY
     InputRange=(X=20,Y=60)
     TwistAxis="0,0,1"
    )) -> (Output_Get:double)
   > Set neck_rot_l(neck_rot_l:double=GetAlphaFromReferenceTransformDelta(SkeletalMeshComponent=Get Head_Mesh(), BoneName=head, BaseName=spine_05, DeltaComponentType=EulerX, InputRange=(X=0,Y=60), TwistAxis="0,0,1")) -> (Output_Get:double)
   > Set Right_leg_belt_crouch(Right_leg_belt_crouch:double=GetAlphaFromReferenceTransformDelta(SkeletalMeshComponent=Get Skel_Mesh(), BoneName=thigh_r, BaseName=pelvis, InputRange=(X=0,Y=-98), TwistAxis="1,0,0")) -> (Output_Get:double)
   > Set Left_leg_belt_crouch(Left_leg_belt_crouch:float=GetAlphaFromReferenceTransformDelta(SkeletalMeshComponent=Get Skel_Mesh(), BoneName=thigh_l, BaseName=pelvis, InputRange=(X=-20,Y=-50), TwistAxis="1,0,0")) -> (Output_Get:float)
   > Set Right_leg_jump_side(Right_leg_jump_side:float=GetAlphaFromReferenceTransformDelta(
     SkeletalMeshComponent=Get Skel_Mesh()
     BoneName=thigh_r
     BaseName=pelvis
     DeltaComponentType=EulerY
     InputRange=(X=0,Y=45)
     TwistAxis="1,0,0"
    )) -> (Output_Get:float)
   > Set Not Crouching(Not Crouching:bool=NOT Boolean(A=Get bIsCrouching())) -> (Output_Get:bool)
   n4 = GetAlphaFromReferenceTransformDelta(SkeletalMeshComponent=Get Skel_Mesh(), BoneName=upperarm_r, BaseName=clavicle_r, InputRange=(X=45,Y=90), TwistAxis="1,0,0")
   > Set Hand rise_R(Hand rise_R:float=@n4) -> (Output_Get:float)
   n5 = Get bIsSkydiveFloating()
   > Set Is Sky Dive Floating(Is Sky Dive Floating:double=FInterpTo(Current=Get Is Sky Dive Floating(), Target="To Float (Boolean)"(InBool=@n5), DeltaTime=Get Delta Time X(), InterpSpeed=10)) -> (Output_Get:double)
   n6 = Get bIsPlayingEmote()
   > Set Is Playing Emote(Is Playing Emote:double=FInterpTo(Current=Get Is Playing Emote(), Target="To Float (Boolean)"(InBool=@n6), DeltaTime=Get Delta Time X(), InterpSpeed=10)) -> (Output_Get:double)
   n7 = Get bIsMoving2D()
   > Set Is Moving 2D(Is Moving 2D:double=FInterpTo(Current=Get Is Moving 2D(), Target="To Float (Boolean)"(InBool=@n7), DeltaTime=Get Delta Time X(), InterpSpeed=10)) -> (Output_Get:double)
   n8 = Get bIsParachuteOpened()
   > Set Is Parachute Opened(Is Parachute Opened:double=FInterpTo(Current=Get Is Parachute Opened(), Target="To Float (Boolean)"(InBool=@n8), DeltaTime=Get Delta Time X(), InterpSpeed=10)) -> (Output_Get:double)
   n9 = Get bIsSliding()
   > Set Is Siding(Is Siding:double=FInterpTo(Current=Get Is Siding(), Target="To Float (Boolean)"(InBool=@n9), DeltaTime=Get Delta Time X(), InterpSpeed=10)) -> (Output_Get:double)
   n10 = Get bIsMoving2D()
   > Set Keys crouching(Keys crouching:bool=AND Boolean(A=Get bIsCrouching(), B=NOT Boolean(A=@n10))) -> (Output_Get:bool)
   n11 = Get bIsCrouching()
   > Set Is Crouching(Is Crouching:double=FInterpTo(Current=Get Is Crouching(), Target="To Float (Boolean)"(InBool=@n11), DeltaTime=Get Delta Time X(), InterpSpeed=10)) -> (Output_Get:double)
   n12 = Get bIsPlayingMeleeAnim()
   > Set Is Playing Melee Anim(Is Playing Melee Anim:double=FInterpTo(Current=Get Is Playing Melee Anim(), Target="To Float (Boolean)"(InBool=@n12), DeltaTime=Get Delta Time X(), InterpSpeed=10)) -> (Output_Get:double)
   n13 = Get bIsFalling()
   n14 = Get bIsJumping()
   > Set Not Crouching and jumping(Not Crouching and jumping:bool=AND Boolean!(A=Get Not Crouching(), B=NOT Boolean(A=@n14), C:bool=NOT Boolean(A=@n13)) -> (ReturnValue)) -> (Output_Get:bool)
   n15 = Get bIsJumping()
   n16 = Get bIsTargeting()
   > Set Bandana idle(Bandana idle:bool=AND Boolean!(A=NOT Boolean(A=@n16), B=NOT Boolean(A=@n15)) -> (ReturnValue)) -> (Output_Get:bool)
   > Set Necklace Alpha(Necklace Alpha:float=GetAlphaFromReferenceTransformDelta(SkeletalMeshComponent=Get Skel_Mesh(), BoneName=head, BaseName=spine_05, InputRange=(X=35,Y=45), OutputRange=(X=1,Y=0), TwistAxis="1,0,0")) -> (Output_Get:float)
   > "Set Necklace  ScubaBond translate"("Necklace  ScubaBond translate":float=float - float(A=1, B=Get Necklace Alpha())) -> (Output_Get:float)
   Event BlueprintInitializeAnimation()
   > Set Skel_Mesh(Skel_Mesh=GetOwnerSkeletalMeshComponent())
   > n17 = Cast To PlayerPawn_Athena(Object=TryGetPawnOwner())
   > n18 = Set Pawn Owner(Pawn Owner:Fort Player Pawn Athena Object Reference=@n17) -> (Output_Get:Fort Player Pawn Athena Object Reference)
   > Set Shoulder_Mesh(Shoulder_Mesh:Skeletal Mesh Component Object Reference=GetSkeletalMeshForPartType(self=@n18)) -> (Output_Get:Skeletal Mesh Component Object Reference)
   > Set Head_Mesh(Head_Mesh:Skeletal Mesh Component Object Reference=GetSkeletalMeshForPartType(self=@n18, PartType=Head)) -> (Output_Get:Skeletal Mesh Component Object Reference)
   n19 = Reroute Node(InputPin:Exec) -> (OutputPin:Exec)
   n20 = Set L_Hand rise_x(
    execute=[@n19.OutputPin, @n26.then]
    L_Hand rise_x:float=GetAlphaFromReferenceTransformDelta(
     SkeletalMeshComponent=Get Skel_Mesh()
     BoneName=upperarm_l
     BaseName=clavicle_l
     DeltaComponentType=EulerX
     InputRange=(X=-50,Y=50)
     TwistAxis="0,0,1"
    )
   ) -> (Output_Get:float)
   n21 = Get L_Hand rise_y()
   n22 = Get L_Hand rise_z() -> (L_Hand rise_z:float)
   n23 = SelectFloat(A=@n21, B=@n22, bPickA="float > float"(A=@n21, B=@n22))
   n24 = Get L_Hand rise_x() -> (L_Hand rise_x:float)
   Set Rand_TEST(execute=[@n25.then_1, @n20.then], Rand_TEST:double=SelectFloat(A=@n23, B=@n24, bPickA="float > float"(A=@n23, B=@n24))) -> (Output_Get:double)
   Reroute Node()
   > n25 = Sequence()
   > PrintString!(
    execute
    InString="To String (Float)"(InDouble=Reroute Node(InputPin:double) -> (OutputPin:double))
    bPrintToScreen
    bPrintToLog
    TextColor
    Duration=0
    Key
   ) -> (then)
   Set L_Hand rise_z(L_Hand rise_z:float=GetAlphaFromReferenceTransformDelta(SkeletalMeshComponent=Get Skel_Mesh(), BoneName=upperarm_l, BaseName=clavicle_l, InputRange=(X=60,Y=80), TwistAxis="0,0,1")) -> (Output_Get:float)
   > n26 = Set L_Hand rise_y(L_Hand rise_y:float=GetAlphaFromReferenceTransformDelta(
     SkeletalMeshComponent=Get Skel_Mesh()
     BoneName=upperarm_l
     BaseName=clavicle_l
     DeltaComponentType=EulerY
     InputRange=(X=60,Y=80)
     TwistAxis="0,0,1"
    )) -> (Output_Get:float)
   OR Boolean(A=false, B=false)
   Branch()
   float + float(A:Wildcard, B:Wildcard) -> (ReturnValue:Wildcard)
   "float > float"(A=Get L_Hand rise_z(), B=Get L_Hand rise_y())
   "Timespan > Timespan"(A:Wildcard, B:Wildcard) -> (ReturnValue:bool=~) {is=title K2Node_PromotableOperator}
   float * float(A=@n4.RawOutput, B=-0.2)
   Get L_Hand rise_y()
  ]}
]
variables [
 Bandana idle {type=boolean}
 calf_l_front {type=number}
 calf_r_front {type=number}
 Delta Time X {type=number, default=0.0177}
 Hand rise_R {type=number}
 head rot {type=Rotator}
 Head_Mesh {type=SkeletalMeshComponent}
 Helmet_on {type=boolean}
 Is Crouching {type=number}
 Is Moving 2D {type=number}
 Is Parachute Opened {type=number}
 Is Playing Emote {type=number}
 Is Playing Melee Anim {type=number}
 Is Siding {type=number}
 Is Sky Dive Floating {type=number}
 isHelmet {type=string, default=M_MED_QuickBurst_PlainsHelmet_FaceAcc}
 Keys crouching {type=boolean}
 L_Hand rise_x {type=number}
 L_Hand rise_y {type=number}
 L_Hand rise_z {type=number}
 Left_leg_belt_crouch {type=number}
 Neck Rot Up Helmet {type=number}
 neck_rot_l {type=number}
 neck_rot_r {type=number}
 neck_rot_up {type=number}
 "Necklace  ScubaBond translate" {type=number}
 Necklace Alpha {type=number}
 Not Crouching {type=boolean, default=true}
 Not Crouching and jumping {type=boolean}
 Pawn Owner {type=FortPlayerPawnAthena}
 Rand_TEST {type=number}
 Right_leg_belt_crouch {type=number}
 Right_leg_jump_side {type=number}
 Shoulder_Mesh {type=SkeletalMeshComponent}
 Shoulder_rot_f_l {type=number}
 Skel_Mesh {type=SkeletalMeshComponent}
 X percentage {type=number, default=0.1}
 Y percentage {type=number, default=0.1}
 Z percentage {type=number, default=0.1}
]
